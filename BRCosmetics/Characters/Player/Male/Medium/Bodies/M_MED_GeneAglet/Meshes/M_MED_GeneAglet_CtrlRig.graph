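# Control Rig for the M_MED_GeneAglet player body that corrects deformation and positions robe, sash, cloak and lightsaber bones as the body moves.
# A sequence runs a shared DeformRig function, builds twist-free thigh helpers via swing-twist, then constrains underskirt, sash and saber.
# Spherical pose readers on legs and torso then weight corrective offsets and set thigh-raise curves.
# A library function traces below the pelvis to place a Kalman-smoothed, slope-aligned skirt ground-collision bone.
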ControlRigBlueprint /BRCosmetics/Characters/Player/Male/Medium/Bodies/M_MED_GeneAglet/Meshes/M_MED_GeneAglet_CtrlRig
graphs [
 New Function_1_ContainedGraph {nodes [
   n1 = RigVMFunctionEntryNode()
   n2 = &RigUnit_GetTransform(Item=(Type=Bone,Name="pelvis"), bInitial=False)
   n3 = RigVMDispatchNode(Elements=(X=0,Y=0,Z=200), Elements.X=@n2.Transform.Translation.X, Elements.Y=@n2.Transform.Translation.Y) {template="DISPATCH_RigVMDispatch_MakeStruct(in Elements,out Struct)"}
   Add = /Script/RigVM.RigVMFunction_MathVectorAdd(B=(X=0,Y=0,Z=-500), A=@n3.Struct) {is=struct RigVMUnitNode, template="Add::Execute(in A,in B,out Result)"}
   n4 = /Script/ControlRig.RigUnit_SphereTraceByTraceChannel(Radius=5, TraceChannel=TraceTypeQuery1, End=@Add.Result, Start=@n3.Struct) {is=struct RigVMUnitNode}
   n5 = /Script/ControlRig.RigUnit_AimBoneMath(
    DebugSettings=(bEnabled=False,Scale=10,WorldOffset=(Rotation=(X=0,Y=0,Z=0,W=1),Translation=(X=0,Y=0,Z=0),Scale3D=(X=1,Y=1,Z=1)))
    InputTransform=(Rotation=(X=0,Y=0,Z=0,W=1),Translation=(X=0,Y=0,Z=0),Scale3D=(X=1,Y=1,Z=1))
    Primary=(Weight=1,Axis=(X=0,Y=0,Z=1),Target=(X=1,Y=0,Z=0),Kind=Location,Space=(Type=Bone,Name="None"))
    Secondary=(Weight=1,Axis=(X=1,Y=0,Z=0),Target=(X=1,Y=0,Z=0),Kind=Direction,Space=(Type=Bone,Name="None"))
    Weight=1
    Primary.Target=@n4.HitNormal
   ) {is=struct RigVMUnitNode}
   RigVMFunction_MathTransformMake = /Script/RigVM.RigVMFunction_MathTransformMake(Scale=(X=1,Y=1,Z=1), Rotation=@n5.Result.Rotation, Translation=@n4.HitLocation) {is=struct RigVMUnitNode}
   KalmanFilter = /Script/RigVM.RigVMFunction_KalmanTransform(BufferSize=16, Value=@RigVMFunction_MathTransformMake.Result) {is=struct RigVMUnitNode, template="KalmanFilter::Execute(in Value,in BufferSize,out Result)"}
   > n6 = &RigUnit_ModifyTransforms(
    ItemToModify=((Item=(Type=Bone,Name="dyn_skirt_ground_collision"),Transform=(Rotation=(X=0,Y=0,Z=0,W=1),Translation=(X=0,Y=0,Z=0),Scale3D=(X=1,Y=1,Z=1))))
    Weight=1
    WeightMinimum=0
    ItemToModify.0.Transform=@KalmanFilter.Result
   )
   RigVMFunction_ControlFlowBranch = &RigVMFunction_ControlFlowBranch(ExecuteContext=@n6.ExecutePin, Condition=@n1.Debug)
   RigVMFunctionReturnNode(ExecuteContext=@RigVMFunction_ControlFlowBranch.Completed)
   RigVMFunction_DebugLineNoSpace = /Script/RigVM.RigVMFunction_DebugLineNoSpace(
    Color=(R=1,G=0,B=0,A=1)
    DebugDrawSettings=(DepthPriority=SDPG_Foreground,Lifetime=-1)
    Thickness=1
    WorldOffset=(Rotation=(X=0,Y=0,Z=0,W=1),Translation=(X=0,Y=0,Z=0),Scale3D=(X=1,Y=1,Z=1))
    bEnabled=True
    B=RigVMDispatchNode(Struct=@RigVMFunction_MathTransformMake.Result) {template="DISPATCH_RigVMDispatch_BreakStruct(in Struct,out Elements)"}.Elements.Translation
    ExecutePin=@RigVMFunction_ControlFlowBranch.True
    A=@n2.Transform.Translation
   ) {is=struct RigVMUnitNode}
   > RigVMFunction_DebugRectangleNoSpace = /Script/RigVM.RigVMFunction_DebugRectangleNoSpace(
    Color=(R=1,G=0,B=0,A=1)
    DebugDrawSettings=(DepthPriority=SDPG_Foreground,Lifetime=-1)
    Scale=200
    Thickness=0
    WorldOffset=(Rotation=(X=0,Y=0,Z=0,W=1),Translation=(X=0,Y=0,Z=0),Scale3D=(X=1,Y=1,Z=1))
    bEnabled=True
    Transform=@RigVMFunction_MathTransformMake.Result
   ) {is=struct RigVMUnitNode}
  ]}
 RigVMFunctionLibrary {nodes [Ground Sloped Collision = RigVMCollapseNode(
    Collision Bone=(Item=(Type=Bone,Name="dyn_skirt_ground_collision"),Transform=(Rotation=(X=0,Y=0,Z=0,W=1),Translation=(X=0,Y=0,Z=0),Scale3D=(X=1,Y=1,Z=1)))
    Debug=False
    ExecuteContext=()
   )]}
 RigVMFunction_Sequence_1_ContainedGraph {nodes [
   n1 = RigVMFunctionEntryNode()
   RigVMFunction_Sequence = &RigVMFunction_Sequence(ExecuteContext=@n1.ExecuteContext)
   n2 = &RigVMFunction_Sequence(ExecuteContext=@RigVMFunction_Sequence.B)
   n3 = &RigVMFunction_Sequence(ExecuteContext=@n2.B)
   n4 = &RigVMFunction_Sequence(ExecuteContext=@n3.B)
   RigVMFunctionReturnNode(A=@RigVMFunction_Sequence.A, B=@n2.A, C=@n3.A, D=@n4.A, E=@n4.B)
  ]}
 RigVMModel {nodes [
   &RigUnit_ParentConstraint(
    AdvancedSettings=(InterpolationType=Average,RotationOrderForFilter=XZY)
    Child=(Type=Bone,Name="dyn_underskirt_A_l_MSTR")
    ExecutePin=()
    Parents=((Item=(Type=Bone,Name="thigh_l"),Weight=1))
    Weight=1
   )
   > &RigUnit_ParentConstraint(
    AdvancedSettings=(InterpolationType=Average,RotationOrderForFilter=XZY)
    Child=(Type=Bone,Name="dyn_underskirt_B_l_MSTR")
    Parents=((Item=(Type=Bone,Name="thigh_l"),Weight=1))
    Weight=1
   )
   > &RigUnit_ParentConstraint(
    AdvancedSettings=(InterpolationType=Average,RotationOrderForFilter=XZY)
    Child=(Type=Bone,Name="dyn_underskirt_C_l_MSTR")
    Parents=((Item=(Type=Bone,Name="thigh_l"),Weight=1))
    Weight=0.5
   )
   > &RigUnit_ParentConstraint(
    AdvancedSettings=(InterpolationType=Average,RotationOrderForFilter=XZY)
    Child=(Type=Bone,Name="dyn_underskirt_C_l_03")
    Filter=(TranslationFilter=(bX=false,bY=false,bZ=false),RotationFilter=(bX=True,bY=True,bZ=True),ScaleFilter=(bX=True,bY=True,bZ=True))
    Parents=((Item=(Type=Bone,Name="calf_l"),Weight=1))
    Weight=0.1
   )
   > &RigUnit_ParentConstraint(
    AdvancedSettings=(InterpolationType=Average,RotationOrderForFilter=XZY)
    Child=(Type=Bone,Name="dyn_underskirt_C_l_02")
    Filter=(TranslationFilter=(bX=false,bY=false,bZ=false),RotationFilter=(bX=True,bY=True,bZ=True),ScaleFilter=(bX=True,bY=True,bZ=True))
    Parents=((Item=(Type=Bone,Name="calf_l"),Weight=1))
    Weight=0.2
   )
   > &RigUnit_ParentConstraint(
    AdvancedSettings=(InterpolationType=Average,RotationOrderForFilter=XZY)
    Child=(Type=Bone,Name="dyn_underskirt_A_r_MSTR")
    Parents=((Item=(Type=Bone,Name="thigh_r"),Weight=1))
    Weight=1
   )
   > &RigUnit_ParentConstraint(
    AdvancedSettings=(InterpolationType=Average,RotationOrderForFilter=XZY)
    Child=(Type=Bone,Name="dyn_underskirt_B_r_MSTR")
    Parents=((Item=(Type=Bone,Name="thigh_r"),Weight=1))
    Weight=1
   )
   > &RigUnit_ParentConstraint(
    AdvancedSettings=(InterpolationType=Average,RotationOrderForFilter=XZY)
    Child=(Type=Bone,Name="dyn_underskirt_C_r_MSTR")
    Parents=((Item=(Type=Bone,Name="thigh_r"),Weight=1))
    Weight=0.5
   )
   > &RigUnit_ParentConstraint(
    AdvancedSettings=(InterpolationType=Average,RotationOrderForFilter=XZY)
    Child=(Type=Bone,Name="dyn_underskirt_C_r_03")
    Filter=(TranslationFilter=(bX=false,bY=false,bZ=false),RotationFilter=(bX=True,bY=True,bZ=True),ScaleFilter=(bX=True,bY=True,bZ=True))
    Parents=((Item=(Type=Bone,Name="calf_r"),Weight=1))
    Weight=0.1
   )
   > &RigUnit_ParentConstraint(
    AdvancedSettings=(InterpolationType=Average,RotationOrderForFilter=XZY)
    Child=(Type=Bone,Name="dyn_underskirt_C_r_02")
    Filter=(TranslationFilter=(bX=false,bY=false,bZ=false),RotationFilter=(bX=True,bY=True,bZ=True),ScaleFilter=(bX=True,bY=True,bZ=True))
    Parents=((Item=(Type=Bone,Name="calf_r"),Weight=1))
    Weight=0.2
   )
   &RigUnit_ParentConstraint(
    AdvancedSettings=(InterpolationType=Average,RotationOrderForFilter=XZY)
    Child=(Type=Bone,Name="dyn_sash_l_01")
    ExecutePin=()
    Filter=(TranslationFilter=(bX=true,bY=true,bZ=true),RotationFilter=(bX=True,bY=True,bZ=True),ScaleFilter=(bX=True,bY=True,bZ=True))
    Parents=((Item=(Type=Bone,Name="dyn_underskirt_A_l_01"),Weight=0.5))
    Weight=1
   )
   > &RigUnit_ParentConstraint(
    AdvancedSettings=(InterpolationType=Average,RotationOrderForFilter=XZY)
    Child=(Type=Bone,Name="dyn_sash_l_02")
    Filter=(TranslationFilter=(bX=true,bY=true,bZ=true),RotationFilter=(bX=True,bY=True,bZ=True),ScaleFilter=(bX=True,bY=True,bZ=True))
    Parents=((Item=(Type=Bone,Name="dyn_underskirt_A_l_02"),Weight=0.5))
    Weight=1
   )
   > &RigUnit_ParentConstraint(
    AdvancedSettings=(InterpolationType=Average,RotationOrderForFilter=XZY)
    Child=(Type=Bone,Name="dyn_sash_l_03")
    Filter=(TranslationFilter=(bX=true,bY=true,bZ=true),RotationFilter=(bX=True,bY=True,bZ=True),ScaleFilter=(bX=True,bY=True,bZ=True))
    Parents=((Item=(Type=Bone,Name="dyn_underskirt_A_l_03"),Weight=1))
    Weight=1
   )
   > &RigUnit_ParentConstraint(
    AdvancedSettings=(InterpolationType=Average,RotationOrderForFilter=XZY)
    Child=(Type=Bone,Name="dyn_sash_r_01")
    Filter=(TranslationFilter=(bX=true,bY=true,bZ=true),RotationFilter=(bX=True,bY=True,bZ=True),ScaleFilter=(bX=True,bY=True,bZ=True))
    Parents=((Item=(Type=Bone,Name="dyn_underskirt_A_r_01"),Weight=0.5))
    Weight=1
   )
   > &RigUnit_ParentConstraint(
    AdvancedSettings=(InterpolationType=Average,RotationOrderForFilter=XZY)
    Child=(Type=Bone,Name="dyn_sash_r_02")
    Filter=(TranslationFilter=(bX=true,bY=true,bZ=true),RotationFilter=(bX=True,bY=True,bZ=True),ScaleFilter=(bX=True,bY=True,bZ=True))
    Parents=((Item=(Type=Bone,Name="dyn_underskirt_A_r_02"),Weight=0.5))
    Weight=1
   )
   > &RigUnit_ParentConstraint(
    AdvancedSettings=(InterpolationType=Average,RotationOrderForFilter=XZY)
    Child=(Type=Bone,Name="dyn_sash_r_03")
    Filter=(TranslationFilter=(bX=true,bY=true,bZ=true),RotationFilter=(bX=True,bY=True,bZ=True),ScaleFilter=(bX=True,bY=True,bZ=True))
    Parents=((Item=(Type=Bone,Name="dyn_underskirt_A_r_03"),Weight=1))
    Weight=1
   )
   > &RigUnit_RotationConstraintLocalSpaceOffset(
    AdvancedSettings=(InterpolationType=Average,RotationOrderForFilter=XZY)
    Child=(Type=Bone,Name="dyn_sash_l_MSTR")
    Filter=(bX=True,bY=false,bZ=false)
    Parents=((Item=(Type=Bone,Name="thigh_l"),Weight=1))
    Weight=1
   )
   > &RigUnit_RotationConstraintLocalSpaceOffset(
    AdvancedSettings=(InterpolationType=Average,RotationOrderForFilter=XZY)
    Child=(Type=Bone,Name="dyn_sash_r_MSTR")
    Filter=(bX=True,bY=false,bZ=false)
    Parents=((Item=(Type=Bone,Name="thigh_r"),Weight=1))
    Weight=1
   )
   &RigUnit_ParentConstraint(
    AdvancedSettings=(InterpolationType=Average,RotationOrderForFilter=XZY)
    Child=(Type=Bone,Name="dyn_underskirt_A_l_MSTR")
    ExecutePin=()
    Parents=((Item=(Type=Bone,Name="thigh_l"),Weight=1))
    Weight=1
   )
   > &RigUnit_ParentConstraint(
    AdvancedSettings=(InterpolationType=Average,RotationOrderForFilter=XZY)
    Child=(Type=Bone,Name="dyn_underskirt_B_l_MSTR")
    Parents=((Item=(Type=Bone,Name="thigh_l"),Weight=1))
    Weight=1
   )
   > &RigUnit_ParentConstraint(
    AdvancedSettings=(InterpolationType=Average,RotationOrderForFilter=XZY)
    Child=(Type=Bone,Name="dyn_underskirt_C_l_MSTR")
    Filter=(TranslationFilter=(bX=true,bY=true,bZ=true),RotationFilter=(bX=True,bY=True,bZ=True),ScaleFilter=(bX=True,bY=True,bZ=True))
    Parents=((Item=(Type=Bone,Name="thigh_l"),Weight=1))
    Weight=0.25
    bMaintainOffset=true
   )
   > &RigUnit_ParentConstraint(
    AdvancedSettings=(InterpolationType=Average,RotationOrderForFilter=XZY)
    Child=(Type=Bone,Name="dyn_underskirt_A_r_MSTR")
    Parents=((Item=(Type=Bone,Name="thigh_r"),Weight=1))
    Weight=1
   )
   > &RigUnit_ParentConstraint(
    AdvancedSettings=(InterpolationType=Average,RotationOrderForFilter=XZY)
    Child=(Type=Bone,Name="dyn_underskirt_B_r_MSTR")
    Parents=((Item=(Type=Bone,Name="thigh_r"),Weight=1))
    Weight=1
   )
   > &RigUnit_ParentConstraint(
    AdvancedSettings=(InterpolationType=Average,RotationOrderForFilter=XZY)
    Child=(Type=Bone,Name="dyn_underskirt_C_r_MSTR")
    Parents=((Item=(Type=Bone,Name="thigh_r"),Weight=1))
    Weight=0.25
   )
   > &RigUnit_ParentConstraint(
    AdvancedSettings=(InterpolationType=Average,RotationOrderForFilter=XZY)
    Child=(Type=Bone,Name="dyn_underskirt_D_MSTR")
    Parents=((Item=(Type=Bone,Name="dyn_underskirt_C_l_MSTR"),Weight=0.5),(Item=(Type=Bone,Name="dyn_underskirt_C_r_MSTR"),Weight=0.5))
    Weight=0.25
   )
   > &RigUnit_ParentConstraint(
    AdvancedSettings=(InterpolationType=Average,RotationOrderForFilter=XZY)
    Child=(Type=Bone,Name="dyn_underskirt_front_MSTR")
    Parents=((Item=(Type=Bone,Name="thigh_l"),Weight=0.5),(Item=(Type=Bone,Name="thigh_r"),Weight=0.5))
    Weight=1
   )
   > &RigUnit_ModifyTransforms(
    ItemToModify="((Item=(Type=Bone,Name=\"None\"),Transform=(Rotation=(X=0,Y=0,Z=0,W=1),Translation=(X=0,Y=0,Z=0),Scale3D=(X=1,Y=1,Z=1))),(Item=(Type=Bone,Name=\"dyn_underskirt_A_l_MSTR\"),Transform=(Rotation=(X=-0.1755,Y=0.0626,Z=0.0464,W=0.9814),Translation=(X=0,Y=0,Z=0),Sca ...(truncated)"
    Weight=1
    WeightMinimum=0
   )
   > n1 = &RigUnit_SphericalPoseReader(
    ActiveRegionScaleFactors=(PositiveWidth=4,NegativeWidth=0.6,PositiveHeight=8,NegativeHeight=8)
    ActiveRegionSize=0.05
    Debug=(bDrawDebug=True,bDraw2D=False,bDrawLocalAxes=False,DebugScale=25,DebugSegments=20,DebugThickness=0.25)
    DriverAxis=(X=-1,Y=0,Z=0)
    DriverItem=(Type=Bone,Name="thigh_l")
    FalloffRegionScaleFactors=(PositiveWidth=1,NegativeWidth=1,PositiveHeight=1,NegativeHeight=1)
    FalloffSize=0.4
    FlipHeightScaling=False
    FlipWidthScaling=False
    OptionalParentItem=(Type=Bone,Name="pelvis")
    RotationOffset=(X=90,Y=0,Z=0)
   )
   > &RigUnit_ModifyTransforms(
    ItemToModify="((Item=(Type=Bone,Name=\"dyn_underskirt_front_MSTR\"),Transform=(Rotation=(X=0,Y=0,Z=0,W=1),Translation=(X=0,Y=0,Z=0),Scale3D=(X=1,Y=1,Z=1))),(Item=(Type=Bone,Name=\"dyn_underskirt_A_r_MSTR\"),Transform=(Rotation=(X=0,Y=0,Z=0,W=1),Translation=(X=0,Y=0,Z=0),Sca ...(truncated)"
    WeightMinimum=0
    Weight=@n1.OutputParam
   )
   > &RigUnit_SphericalPoseReader(
    ActiveRegionScaleFactors=(PositiveWidth=4,NegativeWidth=0.6,PositiveHeight=8,NegativeHeight=8)
    ActiveRegionSize=0.05
    Debug=(bDrawDebug=true,bDraw2D=False,bDrawLocalAxes=False,DebugScale=25,DebugSegments=20,DebugThickness=0.25)
    DriverAxis=(X=-1,Y=0,Z=0)
    DriverItem=(Type=Bone,Name="thigh_l")
    FalloffRegionScaleFactors=(PositiveWidth=1,NegativeWidth=1,PositiveHeight=1,NegativeHeight=1)
    FalloffSize=0.4
    FlipHeightScaling=False
    FlipWidthScaling=False
    OptionalParentItem=(Type=Bone,Name="pelvis")
    RotationOffset=(X=90,Y=0,Z=0)
   )
   &RigUnit_ParentConstraint(
    AdvancedSettings=(InterpolationType=Average,RotationOrderForFilter=XZY)
    Child=(Type=Bone,Name="dyn_sleeve_top_l_MSTR")
    ExecutePin=()
    Filter=(TranslationFilter=(bX=false,bY=false,bZ=false),RotationFilter=(bX=True,bY=True,bZ=True),ScaleFilter=(bX=false,bY=false,bZ=false))
    Parents=((Item=(Type=Bone,Name="hand_l"),Weight=1))
    Weight=0.5
   )
   > &RigUnit_ParentConstraint(
    AdvancedSettings=(InterpolationType=Average,RotationOrderForFilter=XZY)
    Child=(Type=Bone,Name="dyn_sleeve_bottom_l_MSTR")
    Filter=(TranslationFilter=(bX=true,bY=true,bZ=true),RotationFilter=(bX=True,bY=True,bZ=True),ScaleFilter=(bX=false,bY=false,bZ=false))
    Parents=((Item=(Type=Bone,Name="hand_l"),Weight=1))
    Weight=0.5
    bMaintainOffset=true
   )
   > &RigUnit_ModifyTransforms(
    ItemToModify="((Item=(Type=Bone,Name=\"dyn_sleeve_top_r_MSTR\"),Transform=(Rotation=(X=0,Y=0,Z=0,W=1),Translation=(X=0,Y=1,Z=0),Scale3D=(X=1,Y=1,Z=1))),(Item=(Type=Bone,Name=\"dyn_sleeve_bottom_r_MSTR\"),Transform=(Rotation=(X=0,Y=0,Z=0,W=1),Translation=(X=0,Y=-1,Z=0),Scale ...(truncated)"
    Weight=1
    WeightMinimum=0
   )
   > &RigUnit_ParentConstraint(
    AdvancedSettings=(InterpolationType=Average,RotationOrderForFilter=XZY)
    Child=(Type=Bone,Name="dyn_underskirt_front_MSTR")
    Filter=(TranslationFilter=(bX=false,bY=false,bZ=false),RotationFilter=(bX=True,bY=True,bZ=True),ScaleFilter=(bX=True,bY=True,bZ=True))
    Parents=((Item=(Type=Bone,Name="dyn_underskirt_A_r_MSTR"),Weight=0.5),(Item=(Type=Bone,Name="dyn_underskirt_A_l_MSTR"),Weight=0.5))
    Weight=1
   )
   &RigUnit_ParentConstraint(
    AdvancedSettings=(RotationOrderForFilter=XZY)
    Child=(Type=Bone,Name="deform_canister_01_r")
    ExecutePin=()
    Parents="((Item=(Type=Bone,Name=\"thigh_twist_01_r\"),Weight=0.0113),(Item=(Type=Bone,Name=\"deform_groin_r\"),Weight=0.0147),(Item=(Type=Bone,Name=\"spine_03\"),Weight=0.0455),(Item=(Type=Bone,Name=\"spine_01\"),Weight=0.0682),(Item=(Type=Bone,Name=\"spine_02\"),Weight=0.07 ...(truncated)"
    Weight=1
   )
   > &RigUnit_ParentConstraint(
    AdvancedSettings=(RotationOrderForFilter=XZY)
    Child=(Type=Bone,Name="deform_canister_02_r")
    Parents="((Item=(Type=Bone,Name=\"thigh_twist_01_r\"),Weight=0.01),(Item=(Type=Bone,Name=\"deform_groin_r\"),Weight=0.0128),(Item=(Type=Bone,Name=\"spine_03\"),Weight=0.0462),(Item=(Type=Bone,Name=\"spine_01\"),Weight=0.0695),(Item=(Type=Bone,Name=\"spine_02\"),Weight=0.0805 ...(truncated)"
    Weight=1
   )
   > &RigUnit_ParentConstraint(
    AdvancedSettings=(RotationOrderForFilter=XZY)
    Child=(Type=Bone,Name="deform_canister_03_r")
    Parents="((Item=(Type=Bone,Name=\"thigh_twist_01_r\"),Weight=0.0121),(Item=(Type=Bone,Name=\"deform_groin_r\"),Weight=0.0141),(Item=(Type=Bone,Name=\"spine_03\"),Weight=0.0443),(Item=(Type=Bone,Name=\"spine_01\"),Weight=0.0671),(Item=(Type=Bone,Name=\"spine_02\"),Weight=0.07 ...(truncated)"
    Weight=1
   )
   > &RigUnit_ParentConstraint(
    AdvancedSettings=(RotationOrderForFilter=XZY)
    Child=(Type=Bone,Name="deform_canister_04_r")
    Parents="((Item=(Type=Bone,Name=\"thigh_twist_01_r\"),Weight=0.013),(Item=(Type=Bone,Name=\"deform_groin_r\"),Weight=0.0136),(Item=(Type=Bone,Name=\"spine_03\"),Weight=0.0432),(Item=(Type=Bone,Name=\"spine_01\"),Weight=0.0659),(Item=(Type=Bone,Name=\"spine_02\"),Weight=0.076 ...(truncated)"
    Weight=1
   )
   > &RigUnit_ParentConstraint(
    AdvancedSettings=(RotationOrderForFilter=XZY)
    Child=(Type=Bone,Name="deform_pouch_r")
    Parents=((Item=(Type=Bone,Name="deform_pouch_r"),Weight=1.0))
    Weight=1
   )
   > &RigUnit_ParentConstraint(
    AdvancedSettings=(RotationOrderForFilter=XZY)
    Child=(Type=Bone,Name="deform_canister_01_l")
    Parents="((Item=(Type=Bone,Name=\"deform_groin_l\"),Weight=0.0069),(Item=(Type=Bone,Name=\"thigh_twist_01_l\"),Weight=0.0093),(Item=(Type=Bone,Name=\"spine_03\"),Weight=0.0508),(Item=(Type=Bone,Name=\"spine_01\"),Weight=0.0776),(Item=(Type=Bone,Name=\"spine_02\"),Weight=0.09 ...(truncated)"
    Weight=1
   )
   > &RigUnit_ParentConstraint(
    AdvancedSettings=(RotationOrderForFilter=XZY)
    Child=(Type=Bone,Name="deform_canister_02_l")
    Parents="((Item=(Type=Bone,Name=\"deform_groin_l\"),Weight=0.0114),(Item=(Type=Bone,Name=\"thigh_twist_01_l\"),Weight=0.0201),(Item=(Type=Bone,Name=\"spine_03\"),Weight=0.0379),(Item=(Type=Bone,Name=\"spine_01\"),Weight=0.0589),(Item=(Type=Bone,Name=\"spine_02\"),Weight=0.06 ...(truncated)"
    Weight=1
   )
   > &RigUnit_ParentConstraint(
    AdvancedSettings=(RotationOrderForFilter=XZY)
    Child=(Type=Bone,Name="deform_canister_03_l")
    Parents="((Item=(Type=Bone,Name=\"deform_groin_l\"),Weight=0.0087),(Item=(Type=Bone,Name=\"thigh_twist_01_l\"),Weight=0.0185),(Item=(Type=Bone,Name=\"spine_03\"),Weight=0.0383),(Item=(Type=Bone,Name=\"spine_01\"),Weight=0.0596),(Item=(Type=Bone,Name=\"spine_02\"),Weight=0.06 ...(truncated)"
    Weight=1
   )
   > &RigUnit_ParentConstraint(
    AdvancedSettings=(RotationOrderForFilter=XZY)
    Child=(Type=Bone,Name="deform_canister_04_l")
    Parents=((Item=(Type=Bone,Name="thigh_twist_01_l"),Weight=0.0168),(Item=(Type=Bone,Name="spine_03"),Weight=0.0384),(Item=(Type=Bone,Name="spine_01"),Weight=0.0597),(Item=(Type=Bone,Name="spine_02"),Weight=0.0696),(Item=(Type=Bone,Name="pelvis"),Weight=0.8155))
    Weight=1
   )
   > &RigUnit_ParentConstraint(
    AdvancedSettings=(RotationOrderForFilter=XZY)
    Child=(Type=Bone,Name="deform_pouch_l")
    Parents=((Item=(Type=Bone,Name="deform_pouch_l"),Weight=1.0))
    Weight=1
   )
   &RigUnit_ParentConstraint(
    AdvancedSettings=(InterpolationType=Average,RotationOrderForFilter=XZY)
    Child=(Type=Bone,Name="dyn_sleeve_top_r_MSTR")
    ExecutePin=()
    Filter=(TranslationFilter=(bX=false,bY=false,bZ=false),RotationFilter=(bX=True,bY=True,bZ=True),ScaleFilter=(bX=false,bY=false,bZ=false))
    Parents=((Item=(Type=Bone,Name="hand_r"),Weight=1))
    Weight=0.4
    bMaintainOffset=true
   )
   > &RigUnit_ParentConstraint(
    AdvancedSettings=(InterpolationType=Average,RotationOrderForFilter=XZY)
    Child=(Type=Bone,Name="dyn_sleeve_bottom_r_MSTR")
    Filter=(TranslationFilter=(bX=true,bY=true,bZ=true),RotationFilter=(bX=True,bY=True,bZ=True),ScaleFilter=(bX=false,bY=false,bZ=false))
    Parents=((Item=(Type=Bone,Name="hand_r"),Weight=1))
    Weight=0.75
    bMaintainOffset=true
   )
   RigUnit_BeginExecution = &RigUnit_BeginExecution()
   RigVMFunction_Sequence = RigVMAggregateNode(ExecuteContext=@RigUnit_BeginExecution.ExecutePin)
   FN_DeformRig_FNC = RigVMFunctionReferenceNode(
    DeformRig_ Curve="(deformRig_l_upperarm_up_alpha_72_#1=False,deformRig_r_upperarm_up_alpha_73_#2=False,deformRig_l_upperarm_fr_alpha_74_#3=False,deformRig_r_upperarm_fr_alpha_75_CB51B8 ...(truncated)"
    DeformRig_Bool="(upperarm_1_#4=True,elbow_3_#5=True,wrist_7_#6=True,groin_8_#7=True,glute_17_#8=True,knee_9_40C408324CA261 ...(truncated)"
    DeformRig_Debug=(upperarm_debug_18_#4=False,wrist_debug_20_#6=False,lat_debug_26_#9=False,groinGlute_debug_27_#7=False)
    DeformRig_Value="(upperarm_up_18_#10=(Rotation=(X=0,Y=0,Z=0,W=1),Translation=(X=0,Y=0,Z=8),Scale3D=(X=0.7,Y=1,Z=1)),upperarm_fwd_fr_19_#11=(Rotation=(X=0,Y=0,Z=0,W=1),Translation=(X=0,Y=10,Z=0),Scale3D=(X=0.5,Y=1,Z= ...(truncated)"
    ExecuteContext=@RigVMFunction_Sequence.A
   )
   RigUnit_GetTransform = &RigUnit_GetTransform(Item=(Type=Bone,Name="thigh_l"), Space=LocalSpace, bInitial=False)
   RigVMFunction_MathQuaternionSwingTwist = &RigVMFunction_MathQuaternionSwingTwist(TwistAxis=(X=1,Y=0,Z=0), Input=@RigUnit_GetTransform.Transform.Rotation)
   Inverse = /Script/RigVM.RigVMFunction_MathQuaternionInverse(Value=@RigVMFunction_MathQuaternionSwingTwist.Twist) {is=struct RigVMUnitNode, template="Inverse::Execute(in Value,out Result)"}
   RigUnit_SetRotation = &RigUnit_SetRotation(Item=(Type=Bone,Name="CR_thigh_l_notwist"), Weight=1, bInitial=False, Value=@Inverse.Result, ExecutePin=@RigVMFunction_Sequence.B)
   > &RigUnit_ParentConstraint(
    AdvancedSettings=(InterpolationType=Average,RotationOrderForFilter=XZY)
    Child=(Type=Bone,Name="CR_thigh_r")
    Parents=((Item=(Type=Bone,Name="thigh_r"),Weight=1))
    Weight=1
   )
   n2 = &RigUnit_GetTransform(Item=(Type=Bone,Name="cr_thigh_r"), Space=LocalSpace, bInitial=False)
   > &RigUnit_SetRotation(
    Item=(Type=Bone,Name="CR_thigh_no_twist_r")
    Weight=1
    bInitial=False
    Value=&RigVMFunction_MathQuaternionInverse(Value=&RigVMFunction_MathQuaternionSwingTwist(TwistAxis=(X=-1,Y=0,Z=0), Input=@n2.Transform.Rotation).Twist) {template="Inverse::Execute(in Value,out Result)"}.Result
   )
   > &RigUnit_ParentConstraint(
    AdvancedSettings=(InterpolationType=Average,RotationOrderForFilter=XZY)
    Child=(Type=Bone,Name="dyn_underskirt_B_l_MSTR")
    Filter=(TranslationFilter=(bX=true,bY=true,bZ=true),RotationFilter=(bX=True,bY=True,bZ=True),ScaleFilter=(bX=True,bY=True,bZ=True))
    Parents=((Item=(Type=Bone,Name="CR_thigh_l_notwist"),Weight=1))
    Weight=1
   )
   > &RigUnit_ParentConstraint(
    AdvancedSettings=(InterpolationType=Average,RotationOrderForFilter=ZYX)
    Child=(Type=Bone,Name="dyn_underskirt_B_r_MSTR")
    Filter=(TranslationFilter=(bX=true,bY=true,bZ=true),RotationFilter=(bX=True,bY=True,bZ=True),ScaleFilter=(bX=True,bY=True,bZ=True))
    Parents=((Item=(Type=Bone,Name="CR_thigh_no_twist_r"),Weight=1))
    Weight=1
    bMaintainOffset=true
   )
   > &RigUnit_ParentConstraint(
    AdvancedSettings=(InterpolationType=Average,RotationOrderForFilter=XZY)
    Child=(Type=Bone,Name="dyn_underskirt_A_l_MSTR")
    Filter=(TranslationFilter=(bX=true,bY=true,bZ=true),RotationFilter=(bX=True,bY=True,bZ=True),ScaleFilter=(bX=True,bY=True,bZ=True))
    Parents=((Item=(Type=Bone,Name="CR_thigh_l_notwist"),Weight=1))
    Weight=1
   )
   > &RigUnit_ParentConstraint(
    AdvancedSettings=(InterpolationType=Average,RotationOrderForFilter=XZY)
    Child=(Type=Bone,Name="dyn_underskirt_A_r_MSTR")
    Filter=(TranslationFilter=(bX=true,bY=true,bZ=true),RotationFilter=(bX=True,bY=True,bZ=True),ScaleFilter=(bX=True,bY=True,bZ=True))
    Parents=((Item=(Type=Bone,Name="CR_thigh_no_twist_r"),Weight=1))
    Weight=1
   )
   &RigUnit_ParentConstraint(
    AdvancedSettings=(InterpolationType=Average,RotationOrderForFilter=XZY)
    Child=(Type=Bone,Name="dyn_underskirt_front_MSTR")
    Parents=((Item=(Type=Bone,Name="dyn_underskirt_A_l_MSTR"),Weight=1),(Item=(Type=Bone,Name="dyn_underskirt_A_r_MSTR"),Weight=1))
    Weight=1
    ExecutePin=@RigVMFunction_Sequence.C
   )
   > &RigUnit_RotationConstraintLocalSpaceOffset(
    AdvancedSettings=(InterpolationType=Average,RotationOrderForFilter=XZY)
    Child=(Type=Bone,Name="dyn_underskirt_B_l_MSTR")
    Filter=(bX=True,bY=True,bZ=True)
    Parents=((Item=(Type=Bone,Name="thigh_l"),Weight=1))
    Weight=1
   )
   > &RigUnit_RotationConstraintLocalSpaceOffset(
    AdvancedSettings=(InterpolationType=Average,RotationOrderForFilter=XZY)
    Child=(Type=Bone,Name="dyn_underskirt_B_r_MSTR")
    Filter=(bX=True,bY=True,bZ=True)
    Parents=((Item=(Type=Bone,Name="thigh_r"),Weight=1))
    Weight=1
   )
   > &RigUnit_ParentConstraint(
    AdvancedSettings=(InterpolationType=Average,RotationOrderForFilter=XZY)
    Child=(Type=Bone,Name="dyn_sash_l_MSTR")
    Filter=(TranslationFilter=(bX=false,bY=false,bZ=false),RotationFilter=(bX=True,bY=True,bZ=True),ScaleFilter=(bX=True,bY=True,bZ=True))
    Parents=((Item=(Type=Bone,Name="dyn_underskirt_A_l_MSTR"),Weight=1),(Item=(Type=Bone,Name="dyn_underskirt_front_MSTR"),Weight=1))
    Weight=1
   )
   > &RigUnit_ParentConstraint(
    AdvancedSettings=(InterpolationType=Average,RotationOrderForFilter=XZY)
    Child=(Type=Bone,Name="dyn_sash_r_MSTR")
    Filter=(TranslationFilter=(bX=false,bY=false,bZ=false),RotationFilter=(bX=True,bY=True,bZ=True),ScaleFilter=(bX=True,bY=True,bZ=True))
    Parents=((Item=(Type=Bone,Name="dyn_underskirt_A_r_MSTR"),Weight=1),(Item=(Type=Bone,Name="dyn_underskirt_front_MSTR"),Weight=1))
    Weight=1
   )
   > &RigUnit_ParentConstraint(
    AdvancedSettings=(InterpolationType=Average,RotationOrderForFilter=XZY)
    Child=(Type=Bone,Name="dyn_lightsaber_MSTR")
    Filter=(TranslationFilter=(bX=false,bY=false,bZ=false),RotationFilter=(bX=True,bY=True,bZ=True),ScaleFilter=(bX=false,bY=false,bZ=false))
    Parents=((Item=(Type=Bone,Name="dyn_underskirt_A_l_MSTR"),Weight=1))
    Weight=1
   )
   > &RigUnit_ModifyTransforms(
    ItemToModify=((Item=(Type=Bone,Name="dyn_lightsaber_MSTR"),Transform=(Rotation=(X=0,Y=0.0872,Z=0,W=0.9962),Translation=(X=0,Y=0,Z=0),Scale3D=(X=1,Y=1,Z=1))))
    Weight=1
    WeightMinimum=0
   )
   > &RigUnit_ModifyTransforms(
    ItemToModify="((Item=(Type=Bone,Name=\"dyn_sash_r_MSTR\"),Transform=(Rotation=(X=0.0175,Y=0,Z=0,W=0.9998),Translation=(X=0,Y=0,Z=-2),Scale3D=(X=1,Y=1,Z=1))),(Item=(Type=Bone,Name=\"dyn_sash_l_MSTR\"),Transform=(Rotation=(X=0.0175,Y=0,Z=0,W=0.9998),Translation=(X=0,Y=0,Z=2.5 ...(truncated)"
    Weight=1
    WeightMinimum=0
   )
   n3 = &RigVMFunction_Sequence(ExecuteContext=@RigVMFunction_Sequence.D)
   n4 = &RigUnit_SphericalPoseReader(
    ActiveRegionScaleFactors=(PositiveWidth=3.5,NegativeWidth=1.5,PositiveHeight=3,NegativeHeight=3)
    ActiveRegionSize=0.1
    Debug=(bDrawDebug=false,bDraw2D=false,bDrawLocalAxes=false,DebugScale=25,DebugSegments=20,DebugThickness=0.25)
    DriverAxis=(X=-1,Y=0,Z=0)
    DriverItem=(Type=Bone,Name="thigh_l")
    FalloffRegionScaleFactors=(PositiveWidth=1,NegativeWidth=1,PositiveHeight=1,NegativeHeight=1)
    FalloffSize=0.4
    OptionalParentItem=(Type=Bone,Name="pelvis")
    RotationOffset=(X=120,Y=0,Z=0)
    ExecutePin=@n3.A
   )
   > /Script/ControlRig.RigUnit_SetCurveValue(Curve=thigh_l_up, Value=@n4.OutputParam) {is=struct RigVMUnitNode}
   > &RigUnit_ModifyTransforms(
    ItemToModify="((Item=(Type=Bone,Name=\"dyn_underskirt_front_MSTR\"),Transform=(Rotation=(X=0,Y=0,Z=0.1305,W=0.9914),Translation=(X=0,Y=0,Z=0),Scale3D=(X=1,Y=1,Z=1))),(Item=(Type=Bone,Name=\"dyn_sash_l_MSTR\"),Transform=(Rotation=(X=0.0076,Y=-0.0868,Z=0.0868,W=0.9924),Transl ...(truncated)"
    WeightMinimum=0
    Weight=@n4.OutputParam
   )
   > n5 = &RigUnit_SphericalPoseReader(
    ActiveRegionScaleFactors=(PositiveWidth=3.5,NegativeWidth=1.5,PositiveHeight=3,NegativeHeight=3)
    ActiveRegionSize=0.1
    Debug=(bDrawDebug=false,bDraw2D=false,bDrawLocalAxes=false,DebugScale=25,DebugSegments=20,DebugThickness=0.25)
    DriverAxis=(X=-1,Y=0,Z=0)
    DriverItem=(Type=Bone,Name="thigh_l")
    FalloffRegionScaleFactors=(PositiveWidth=1,NegativeWidth=1,PositiveHeight=1,NegativeHeight=1)
    FalloffSize=0.4
    OptionalParentItem=(Type=Bone,Name="pelvis")
    RotationOffset=(X=120,Y=0,Z=0)
   )
   RigVMFunction_AlphaInterp = &RigVMFunction_AlphaInterp(InterpSpeedDecreasing=10, InterpSpeedIncreasing=10, Scale=1, bClampResult=False, bInterpResult=True, Value=@n5.OutputParam)
   > &RigUnit_ModifyTransforms(
    ItemToModify=((Item=(Type=Bone,Name="dyn_lightsaber_MSTR"),Transform=(Rotation=(X=0.1496,Y=0.1151,Z=-0.4876,W=0.8524),Translation=(X=0,Y=-2.0003,Z=2.0002),Scale3D=(X=1,Y=1,Z=1))))
    WeightMinimum=0
    Weight=@RigVMFunction_AlphaInterp.Result
   )
   > n6 = &RigUnit_SphericalPoseReader(
    ActiveRegionScaleFactors=(PositiveWidth=1,NegativeWidth=1,PositiveHeight=1.5,NegativeHeight=1.5)
    ActiveRegionSize=0.15
    Debug=(bDrawDebug=false,bDraw2D=false,bDrawLocalAxes=false,DebugScale=25,DebugSegments=20,DebugThickness=0.25)
    DriverAxis=(X=-1,Y=0,Z=0)
    DriverItem=(Type=Bone,Name="thigh_l")
    FalloffRegionScaleFactors=(PositiveWidth=1,NegativeWidth=1,PositiveHeight=1,NegativeHeight=1)
    FalloffSize=0.25
    OptionalParentItem=(Type=Bone,Name="pelvis")
    RotationOffset=(X=-90,Y=0,Z=-30)
   )
   > &RigUnit_ModifyTransforms(
    ItemToModify="((Item=(Type=Bone,Name=\"dyn_underskirt_front_MSTR\"),Transform=(Rotation=(X=0,Y=-0.1736,Z=0,W=0.9848),Translation=(X=0,Y=0,Z=0),Scale3D=(X=1,Y=1,Z=1))),(Item=(Type=Bone,Name=\"dyn_underskirt_A_l_MSTR\"),Transform=(Rotation=(X=0,Y=-0.0872,Z=0,W=0.9962),Transla ...(truncated)"
    WeightMinimum=0
    Weight=@n6.OutputParam
   )
   > n7 = &RigUnit_SphericalPoseReader(
    ActiveRegionScaleFactors=(PositiveWidth=1,NegativeWidth=1,PositiveHeight=1,NegativeHeight=1)
    ActiveRegionSize=0.1
    Debug=(bDrawDebug=false,bDraw2D=False,bDrawLocalAxes=False,DebugScale=25,DebugSegments=20,DebugThickness=0.25)
    DriverAxis=(X=0,Y=-1,Z=0)
    DriverItem=(Type=Bone,Name="pelvis")
    FalloffRegionScaleFactors=(PositiveWidth=1,NegativeWidth=1,PositiveHeight=1,NegativeHeight=1)
    FalloffSize=0.2
    FlipHeightScaling=False
    FlipWidthScaling=False
    OptionalParentItem=(Type=Bone,Name="spine_05")
    RotationOffset=(X=-90,Y=0,Z=45)
   )
   > &RigUnit_ModifyTransforms(
    ItemToModify="((Item=(Type=Bone,Name=\"dyn_cloak_C_01_l_MSTR\"),Transform=(Rotation=(X=0,Y=-0.2079,Z=0,W=0.9781),Translation=(X=0,Y=0,Z=2),Scale3D=(X=1,Y=1,Z=1))),(Item=(Type=Bone,Name=\"dyn_cloak_C_01_r_MSTR\"),Transform=(Rotation=(X=0,Y=-0.2079,Z=0,W=0.9781),Translation=( ...(truncated)"
    WeightMinimum=0
    Weight=@n7.OutputParam
   )
   > n8 = &RigUnit_SphericalPoseReader(
    ActiveRegionScaleFactors=(PositiveWidth=1,NegativeWidth=1,PositiveHeight=1,NegativeHeight=1)
    ActiveRegionSize=0.1
    Debug=(bDrawDebug=false,bDraw2D=False,bDrawLocalAxes=False,DebugScale=25,DebugSegments=20,DebugThickness=0.25)
    DriverAxis=(X=-1,Y=0,Z=0)
    DriverItem=(Type=Bone,Name="calf_l")
    FalloffRegionScaleFactors=(PositiveWidth=1,NegativeWidth=1,PositiveHeight=1,NegativeHeight=1)
    FalloffSize=0.2
    FlipHeightScaling=False
    FlipWidthScaling=False
    OptionalParentItem=(Type=Bone,Name="thigh_l")
    RotationOffset=(X=90,Y=0,Z=210)
   )
   > &RigUnit_ModifyTransforms(
    ItemToModify="((Item=(Type=Bone,Name=\"dyn_underskirt_C_l_MSTR\"),Transform=(Rotation=(X=0,Y=0,Z=0,W=1),Translation=(X=0,Y=0,Z=0),Scale3D=(X=0.9,Y=1,Z=1))),(Item=(Type=Bone,Name=\"dyn_underskirt_B_l_MSTR\"),Transform=(Rotation=(X=0,Y=0,Z=0,W=1),Translation=(X=0,Y=0,Z=0),Sca ...(truncated)"
    WeightMinimum=0
    Weight=@n8.OutputParam
   )
   > n9 = &RigUnit_SphericalPoseReader(
    ActiveRegionScaleFactors=(PositiveWidth=3,NegativeWidth=1,PositiveHeight=1.5,NegativeHeight=1.5)
    ActiveRegionSize=0.1
    Debug=(bDrawDebug=false,bDraw2D=False,bDrawLocalAxes=False,DebugScale=25,DebugSegments=20,DebugThickness=0.25)
    DriverAxis=(X=-1,Y=0,Z=0)
    DriverItem=(Type=Bone,Name="thigh_l")
    FalloffRegionScaleFactors=(PositiveWidth=1,NegativeWidth=1,PositiveHeight=1,NegativeHeight=1)
    FalloffSize=0.5
    FlipHeightScaling=False
    FlipWidthScaling=False
    OptionalParentItem=(Type=Bone,Name="pelvis")
    RotationOffset=(X=-125,Y=0,Z=0)
   )
   > &RigUnit_ModifyTransforms(
    ItemToModify="((Item=(Type=Bone,Name=\"dyn_underskirt_B_l_MSTR\"),Transform=(Rotation=(X=-0.0094,Y=0.1069,Z=0.0864,W=-0.9905),Translation=(X=0,Y=0,Z=0),Scale3D=(X=1,Y=1,Z=1))),(Item=(Type=Bone,Name=\"dyn_underskirt_C_l_MSTR\"),Transform=(Rotation=(X=-0.114,Y=-0.506,Z=0.2268 ...(truncated)"
    WeightMinimum=0
    Weight=@n9.OutputParam
   )
   > n10 = &RigUnit_SphericalPoseReader(
    ActiveRegionScaleFactors=(PositiveWidth=3,NegativeWidth=1,PositiveHeight=3,NegativeHeight=3)
    ActiveRegionSize=0.1
    Debug=(bDrawDebug=false,bDraw2D=False,bDrawLocalAxes=False,DebugScale=25,DebugSegments=20,DebugThickness=0.25)
    DriverAxis=(X=-1,Y=0,Z=0)
    DriverItem=(Type=Bone,Name="thigh_l")
    FalloffRegionScaleFactors=(PositiveWidth=1,NegativeWidth=0.95,PositiveHeight=1,NegativeHeight=1)
    FalloffSize=0.5
    FlipHeightScaling=False
    FlipWidthScaling=False
    OptionalParentItem=(Type=Bone,Name="pelvis")
    RotationOffset=(X=-90,Y=0,Z=0)
   )
   > &RigUnit_ModifyTransforms(
    ItemToModify="((Item=(Type=Bone,Name=\"dyn_underskirt_front_MSTR\"),Transform=(Rotation=(X=-0.0151,Y=-0.0858,Z=-0.173,W=0.9811),Translation=(X=0,Y=0,Z=0),Scale3D=(X=1,Y=1,Z=1))),(Item=(Type=Bone,Name=\"dyn_underskirt_C_l_MSTR\"),Transform=(Rotation=(X=0.0275,Y=-0.1855,Z=-0. ...(truncated)"
    WeightMinimum=0
    Weight=@n10.OutputParam
   )
   > n11 = &RigUnit_SphericalPoseReader(
    ActiveRegionScaleFactors=(PositiveWidth=3.5,NegativeWidth=1.5,PositiveHeight=2,NegativeHeight=2)
    ActiveRegionSize=0.1
    Debug=(bDrawDebug=true,bDraw2D=false,bDrawLocalAxes=false,DebugScale=25,DebugSegments=20,DebugThickness=0.25)
    DriverAxis=(X=-1,Y=0,Z=0)
    DriverItem=(Type=Bone,Name="thigh_l")
    FalloffRegionScaleFactors=(PositiveWidth=1,NegativeWidth=1,PositiveHeight=1,NegativeHeight=1)
    FalloffSize=0.3
    OptionalParentItem=(Type=Bone,Name="pelvis")
    RotationOffset=(X=50,Y=0,Z=30)
   )
   > &RigUnit_ModifyTransforms(
    ItemToModify="((Item=(Type=Bone,Name=\"dyn_underskirt_A_l_MSTR\"),Transform=(Rotation=(X=-0.0023,Y=0.0268,Z=0.019,W=-0.9995),Translation=(X=0,Y=0,Z=0),Scale3D=(X=1,Y=1,Z=1))),(Item=(Type=Bone,Name=\"dyn_underskirt_front_MSTR\"),Transform=(Rotation=(X=0.6249,Y=0.1843,Z=-0.07 ...(truncated)"
    WeightMinimum=0
    Weight=@n11.OutputParam
   )
   n12 = &RigUnit_SphericalPoseReader(
    ActiveRegionScaleFactors=(PositiveWidth=3.5,NegativeWidth=1.5,PositiveHeight=3,NegativeHeight=3)
    ActiveRegionSize=0.1
    Debug=(bDrawDebug=false,bDraw2D=false,DebugScale=25,DebugSegments=20,DebugThickness=0.25)
    DriverAxis=(X=1,Y=0,Z=0)
    DriverItem=(Type=Bone,Name="thigh_r")
    FalloffRegionScaleFactors=(PositiveWidth=1,NegativeWidth=1,PositiveHeight=1,NegativeHeight=1)
    FalloffSize=0.4
    OptionalParentItem=(Type=Bone,Name="pelvis")
    RotationOffset=(X=-120,Y=180,Z=0)
    ExecutePin=@n3.B
   )
   > &RigUnit_SetCurveValue(Curve=thigh_r_up, Value=@n12.OutputParam)
   > &RigUnit_ModifyTransforms(
    ItemToModify="((Item=(Type=Bone,Name=\"dyn_underskirt_front_MSTR\"),Transform=(Rotation=(X=0,Y=0,Z=-0.1736,W=0.9848),Translation=(X=0,Y=0,Z=0),Scale3D=(X=1,Y=1,Z=1))),(Item=(Type=Bone,Name=\"dyn_sash_r_MSTR\"),Transform=(Rotation=(X=-0.0868,Y=-0.0868,Z=-0.0076,W=0.9924),Tra ...(truncated)"
    WeightMinimum=0
    Weight=@n12.OutputParam
   )
   > n13 = &RigUnit_SphericalPoseReader(
    ActiveRegionScaleFactors=(PositiveWidth=1,NegativeWidth=1,PositiveHeight=1.5,NegativeHeight=1.5)
    ActiveRegionSize=0.15
    Debug=(bDrawDebug=false,bDraw2D=false,bDrawLocalAxes=false,DebugScale=25,DebugSegments=20,DebugThickness=0.25)
    DriverAxis=(X=1,Y=0,Z=0)
    DriverItem=(Type=Bone,Name="thigh_r")
    FalloffRegionScaleFactors=(PositiveWidth=1,NegativeWidth=1,PositiveHeight=1,NegativeHeight=1)
    FalloffSize=0.25
    OptionalParentItem=(Type=Bone,Name="pelvis")
    RotationOffset=(X=90,Y=0,Z=-30)
   )
   > &RigUnit_ModifyTransforms(
    ItemToModify="((Item=(Type=Bone,Name=\"dyn_underskirt_front_MSTR\"),Transform=(Rotation=(X=0,Y=-0.1736,Z=0,W=0.9848),Translation=(X=0,Y=0,Z=0),Scale3D=(X=1,Y=1,Z=1))),(Item=(Type=Bone,Name=\"dyn_sash_r_MSTR\"),Transform=(Rotation=(X=0.1305,Y=0,Z=0,W=0.9914),Translation=(X=0 ...(truncated)"
    WeightMinimum=0
    Weight=@n13.OutputParam
   )
   > n14 = &RigUnit_SphericalPoseReader(
    ActiveRegionScaleFactors=(PositiveWidth=3,NegativeWidth=1,PositiveHeight=1.3,NegativeHeight=1.2)
    ActiveRegionSize=0.1
    Debug=(bDrawDebug=false,bDraw2D=False,bDrawLocalAxes=False,DebugScale=25,DebugSegments=20,DebugThickness=0.25)
    DriverAxis=(X=1,Y=0,Z=0)
    DriverItem=(Type=Bone,Name="thigh_r")
    FalloffRegionScaleFactors=(PositiveWidth=1,NegativeWidth=1,PositiveHeight=0.5,NegativeHeight=0.5)
    FalloffSize=0.3
    FlipHeightScaling=true
    FlipWidthScaling=true
    OptionalParentItem=(Type=Bone,Name="None")
    RotationOffset=(X=-50,Y=5,Z=-12)
   )
   > &RigUnit_ModifyTransforms(
    ItemToModify="((Item=(Type=Bone,Name=\"dyn_cloak_A_01_r_MSTR\"),Transform=(Rotation=(X=-0.4178,Y=-0.223,Z=0.1562,W=0.8668),Translation=(X=0,Y=0,Z=2),Scale3D=(X=1,Y=1,Z=1))),(Item=(Type=Bone,Name=\"dyn_cloak_B_01_r_MSTR\"),Transform=(Rotation=(X=-0.0449,Y=0.2639,Z=-0.006,W=- ...(truncated)"
    WeightMinimum=0
    Weight=@n14.OutputParam
   )
   > n15 = &RigUnit_SphericalPoseReader(
    ActiveRegionScaleFactors=(PositiveWidth=3,NegativeWidth=1,PositiveHeight=1,NegativeHeight=1)
    ActiveRegionSize=0.2
    Debug=(bDrawDebug=false,bDraw2D=False,bDrawLocalAxes=False,DebugScale=25,DebugSegments=20,DebugThickness=0.25)
    DriverAxis=(X=1,Y=0,Z=0)
    DriverItem=(Type=Bone,Name="thigh_r")
    FalloffRegionScaleFactors=(PositiveWidth=1,NegativeWidth=1,PositiveHeight=1,NegativeHeight=1)
    FalloffSize=0.5
    FlipHeightScaling=true
    FlipWidthScaling=true
    OptionalParentItem=(Type=Bone,Name="None")
    RotationOffset=(X=0,Y=0,Z=0)
   )
   > &RigUnit_ModifyTransforms(
    ItemToModify="((Item=(Type=Bone,Name=\"dyn_cloak_A_01_r_MSTR\"),Transform=(Rotation=(X=0.2495,Y=0.2148,Z=-0.6161,W=-0.7155),Translation=(X=0,Y=0,Z=0),Scale3D=(X=1,Y=1,Z=1))),(Item=(Type=Bone,Name=\"dyn_cloak_B_01_r_MSTR\"),Transform=(Rotation=(X=0,Y=0,Z=-0.4319,W=-0.9019),T ...(truncated)"
    WeightMinimum=0
    Weight=@n15.OutputParam
   )
   > n16 = &RigUnit_SphericalPoseReader(
    ActiveRegionScaleFactors=(PositiveWidth=1,NegativeWidth=1,PositiveHeight=1,NegativeHeight=1)
    ActiveRegionSize=0.15
    Debug=(bDrawDebug=false,bDraw2D=False,bDrawLocalAxes=False,DebugScale=25,DebugSegments=20,DebugThickness=0.25)
    DriverAxis=(X=0,Y=1,Z=0)
    DriverItem=(Type=Bone,Name="spine_05")
    FalloffRegionScaleFactors=(PositiveWidth=1,NegativeWidth=1,PositiveHeight=0.5,NegativeHeight=0.5)
    FalloffSize=0.6
    FlipHeightScaling=False
    FlipWidthScaling=False
    OptionalParentItem=(Type=Bone,Name="pelvis")
    RotationOffset=(X=30,Y=20,Z=0)
   )
   > &RigUnit_ModifyTransforms(
    ItemToModify="((Item=(Type=Bone,Name=\"dyn_cloak_A_01_l_MSTR\"),Transform=(Rotation=(X=0.25,Y=-0.067,Z=-0.25,W=0.933),Translation=(X=0,Y=0,Z=0),Scale3D=(X=1,Y=1,Z=1))),(Item=(Type=Bone,Name=\"dyn_cloak_B_01_l_MSTR\"),Transform=(Rotation=(X=0.236,Y=-0.1441,Z=-0.1402,W=0.9507 ...(truncated)"
    WeightMinimum=0
    Weight=@n16.OutputParam
   )
   > n17 = &RigUnit_SphericalPoseReader(
    ActiveRegionScaleFactors=(PositiveWidth=1,NegativeWidth=1,PositiveHeight=1,NegativeHeight=1)
    ActiveRegionSize=0.1
    Debug=(bDrawDebug=false,bDraw2D=False,bDrawLocalAxes=False,DebugScale=25,DebugSegments=20,DebugThickness=0.25)
    DriverAxis=(X=1,Y=0,Z=0)
    DriverItem=(Type=Bone,Name="calf_r")
    FalloffRegionScaleFactors=(PositiveWidth=1,NegativeWidth=1,PositiveHeight=1,NegativeHeight=1)
    FalloffSize=0.2
    FlipHeightScaling=False
    FlipWidthScaling=False
    OptionalParentItem=(Type=Bone,Name="thigh_r")
    RotationOffset=(X=90,Y=0,Z=48)
   )
   > &RigUnit_ModifyTransforms(
    ItemToModify="((Item=(Type=Bone,Name=\"dyn_underskirt_C_r_MSTR\"),Transform=(Rotation=(X=0,Y=0,Z=0,W=1),Translation=(X=0,Y=0,Z=0),Scale3D=(X=0.5,Y=1,Z=1))),(Item=(Type=Bone,Name=\"dyn_underskirt_B_r_MSTR\"),Transform=(Rotation=(X=0,Y=0,Z=0,W=1),Translation=(X=0,Y=0,Z=0),Sca ...(truncated)"
    WeightMinimum=0
    Weight=@n17.OutputParam
   )
   > &RigUnit_SphericalPoseReader(
    ActiveRegionScaleFactors=(PositiveWidth=3,NegativeWidth=1,PositiveHeight=3,NegativeHeight=3)
    ActiveRegionSize=0.1
    Debug=(bDrawDebug=false,bDraw2D=False,bDrawLocalAxes=False,DebugScale=25,DebugSegments=20,DebugThickness=0.25)
    DriverAxis=(X=1,Y=0,Z=0)
    DriverItem=(Type=Bone,Name="thigh_r")
    FalloffRegionScaleFactors=(PositiveWidth=1,NegativeWidth=0.95,PositiveHeight=1,NegativeHeight=1)
    FalloffSize=0.5
    FlipHeightScaling=true
    FlipWidthScaling=true
    OptionalParentItem=(Type=Bone,Name="None")
    RotationOffset=(X=90,Y=0,Z=0)
   )
   > n18 = &RigUnit_SphericalPoseReader(
    ActiveRegionScaleFactors=(PositiveWidth=3,NegativeWidth=1,PositiveHeight=1.5,NegativeHeight=1.5)
    ActiveRegionSize=0.1
    Debug=(bDrawDebug=false,bDraw2D=False,bDrawLocalAxes=False,DebugScale=25,DebugSegments=20,DebugThickness=0.25)
    DriverAxis=(X=1,Y=0,Z=0)
    DriverItem=(Type=Bone,Name="thigh_r")
    FalloffRegionScaleFactors=(PositiveWidth=1,NegativeWidth=1,PositiveHeight=1,NegativeHeight=1)
    FalloffSize=0.5
    FlipHeightScaling=true
    FlipWidthScaling=true
    OptionalParentItem=(Type=Bone,Name="None")
    RotationOffset=(X=55,Y=0,Z=0)
   )
   > &RigUnit_ModifyTransforms(
    ItemToModify="((Item=(Type=Bone,Name=\"dyn_underskirt_B_r_MSTR\"),Transform=(Rotation=(X=0.0702,Y=0.012,Z=-0.4118,W=0.9085),Translation=(X=0,Y=0,Z=0),Scale3D=(X=1,Y=1,Z=1))),(Item=(Type=Bone,Name=\"dyn_underskirt_C_r_MSTR\"),Transform=(Rotation=(X=0.0052,Y=0.1656,Z=-0.6068, ...(truncated)"
    WeightMinimum=0
    Weight=@n18.OutputParam
   )
   > n19 = &RigUnit_SphericalPoseReader(
    ActiveRegionScaleFactors=(PositiveWidth=3,NegativeWidth=1,PositiveHeight=3,NegativeHeight=3)
    ActiveRegionSize=0.1
    Debug=(bDrawDebug=false,bDraw2D=False,bDrawLocalAxes=False,DebugScale=25,DebugSegments=20,DebugThickness=0.25)
    DriverAxis=(X=1,Y=0,Z=0)
    DriverItem=(Type=Bone,Name="thigh_r")
    FalloffRegionScaleFactors=(PositiveWidth=1,NegativeWidth=0.95,PositiveHeight=1,NegativeHeight=1)
    FalloffSize=0.5
    FlipHeightScaling=true
    FlipWidthScaling=true
    OptionalParentItem=(Type=Bone,Name="None")
    RotationOffset=(X=90,Y=0,Z=0)
   )
   > &RigUnit_ModifyTransforms(
    ItemToModify="((Item=(Type=Bone,Name=\"dyn_underskirt_front_MSTR\"),Transform=(Rotation=(X=0.0107,Y=0.0865,Z=-0.122,W=-0.9887),Translation=(X=0,Y=0,Z=0),Scale3D=(X=1,Y=1,Z=1))),(Item=(Type=Bone,Name=\"dyn_underskirt_C_r_MSTR\"),Transform=(Rotation=(X=0.1556,Y=0.5302,Z=0.308 ...(truncated)"
    WeightMinimum=0
    Weight=@n19.OutputParam
   )
   > n20 = &RigUnit_SphericalPoseReader(
    ActiveRegionScaleFactors=(PositiveWidth=3.5,NegativeWidth=1.5,PositiveHeight=2,NegativeHeight=2)
    ActiveRegionSize=0.1
    Debug=(bDrawDebug=true,bDraw2D=false,DebugScale=25,DebugSegments=20,DebugThickness=0.25)
    DriverAxis=(X=1,Y=0,Z=0)
    DriverItem=(Type=Bone,Name="thigh_r")
    FalloffRegionScaleFactors=(PositiveWidth=1,NegativeWidth=1,PositiveHeight=1,NegativeHeight=1)
    FalloffSize=0.3
    OptionalParentItem=(Type=Bone,Name="pelvis")
    RotationOffset=(X=-50,Y=210,Z=0)
   )
   > &RigUnit_ModifyTransforms(
    ItemToModify="((Item=(Type=Bone,Name=\"dyn_underskirt_front_MSTR\"),Transform=(Rotation=(X=-0.2435,Y=0.1757,Z=0.1331,W=-0.9445),Translation=(X=0,Y=0,Z=0),Scale3D=(X=1,Y=1,Z=1))),(Item=(Type=Bone,Name=\"dyn_underskirt_A_l_MSTR\"),Transform=(Rotation=(X=-0.1502,Y=0.149,Z=0.12 ...(truncated)"
    WeightMinimum=0
    Weight=@n20.OutputParam
   )
   RigVMFunction_ControlFlowBranch = &RigVMFunction_ControlFlowBranch(ExecuteContext=@RigVMFunction_Sequence.E, Condition=RigVMVariableNode() {variable=Disable Floor Slope}.Value)
   KalmanFilter = &RigVMFunction_KalmanFloat(BufferSize=16, Value=RigVMVariableNode() {variable=Push Down Weight}.Value) {template="KalmanFilter::Execute(in Value,in BufferSize,out Result)"}
   &RigUnit_ModifyTransforms(
    ItemToModify=((Item=(Type=Bone,Name="dyn_skirt_ground_collision"),Transform=(Rotation=(X=0,Y=0,Z=0,W=1),Translation=(X=0,Y=0,Z=-500),Scale3D=(X=1,Y=1,Z=1))))
    WeightMinimum=0
    Weight=@KalmanFilter.Result
    ExecutePin=@RigVMFunction_ControlFlowBranch.Completed
   )
   RigVMVariableNode(Value=0, ExecuteContext=@RigVMFunction_ControlFlowBranch.False) {variable=Push Down Weight}
   RigVMVariableNode(Value=1, ExecuteContext=@RigVMFunction_ControlFlowBranch.True) {variable=Push Down Weight}
   n21 = &RigUnit_SphericalPoseReader(
    ActiveRegionScaleFactors=(PositiveWidth=1,NegativeWidth=1,PositiveHeight=8,NegativeHeight=8)
    ActiveRegionSize=0.05
    Debug=(bDrawDebug=True,bDraw2D=False,bDrawLocalAxes=False,DebugScale=25,DebugSegments=20,DebugThickness=0.25)
    DriverAxis=(X=1,Y=0,Z=0)
    DriverItem=(Type=Bone,Name="thigh_r")
    ExecutePin=()
    FalloffRegionScaleFactors=(PositiveWidth=0.7,NegativeWidth=0.7,PositiveHeight=1,NegativeHeight=1)
    FalloffSize=0.4
    FlipHeightScaling=False
    FlipWidthScaling=False
    OptionalParentItem=(Type=Bone,Name="pelvis")
    RotationOffset=(X=90,Y=0,Z=180)
   )
   > &RigUnit_ModifyTransforms(
    ItemToModify="((Item=(Type=Bone,Name=\"dyn_underskirt_front_MSTR\"),Transform=(Rotation=(X=0,Y=-0.1305,Z=0,W=0.9914),Translation=(X=0,Y=0,Z=4),Scale3D=(X=1,Y=1,Z=1))),(Item=(Type=Bone,Name=\"dyn_underskirt_A_r_MSTR\"),Transform=(Rotation=(X=0,Y=-0.0523,Z=0,W=0.9986),Transla ...(truncated)"
    WeightMinimum=0
    Weight=@n21.OutputParam
   )
   &RigUnit_SphericalPoseReader(
    ActiveRegionScaleFactors=(PositiveWidth=1,NegativeWidth=1,PositiveHeight=1,NegativeHeight=1)
    ActiveRegionSize=0.05
    Debug=(bDrawDebug=true,bDraw2D=False,bDrawLocalAxes=False,DebugScale=25,DebugSegments=20,DebugThickness=0.25)
    DriverAxis=(X=-1,Y=0,Z=0)
    DriverItem=(Type=Bone,Name="thigh_l")
    ExecutePin=()
    FalloffRegionScaleFactors=(PositiveWidth=1,NegativeWidth=1,PositiveHeight=1,NegativeHeight=1)
    FalloffSize=0.2
    FlipHeightScaling=False
    FlipWidthScaling=False
    OptionalParentItem=(Type=Bone,Name="pelvis")
    RotationOffset=(X=90,Y=0,Z=90)
   )
   > &RigUnit_ModifyTransforms(
    ItemToModify=((Item=(Type=Bone,Name="dyn_underskirt_front_MSTR"),Transform=(Rotation=(X=0,Y=-0.0872,Z=0,W=0.9962),Translation=(X=0,Y=0,Z=0),Scale3D=(X=1,Y=1,Z=1))))
    Weight=1
    WeightMinimum=0
   )
   > &RigUnit_SphericalPoseReader(
    ActiveRegionScaleFactors=(PositiveWidth=1,NegativeWidth=1,PositiveHeight=1,NegativeHeight=1)
    ActiveRegionSize=0.05
    Debug=(bDrawDebug=true,bDraw2D=False,bDrawLocalAxes=False,DebugScale=25,DebugSegments=20,DebugThickness=0.25)
    DriverAxis=(X=1,Y=0,Z=0)
    DriverItem=(Type=Bone,Name="thigh_r")
    FalloffRegionScaleFactors=(PositiveWidth=1,NegativeWidth=1,PositiveHeight=1,NegativeHeight=1)
    FalloffSize=0.2
    FlipHeightScaling=False
    FlipWidthScaling=False
    OptionalParentItem=(Type=Bone,Name="pelvis")
    RotationOffset=(X=90,Y=0,Z=-70)
   )
   > &RigUnit_ModifyTransforms(
    ItemToModify=((Item=(Type=Bone,Name="dyn_underskirt_front_MSTR"),Transform=(Rotation=(X=0,Y=-0.0872,Z=0,W=0.9962),Translation=(X=0,Y=0,Z=0),Scale3D=(X=1,Y=1,Z=1))))
    Weight=1
    WeightMinimum=0
   )
   n22 = &RigUnit_SphericalPoseReader(
    ActiveRegionScaleFactors=(PositiveWidth=1,NegativeWidth=1,PositiveHeight=1,NegativeHeight=1)
    ActiveRegionSize=0.2
    Debug=(bDrawDebug=false,bDraw2D=False,bDrawLocalAxes=False,DebugScale=25,DebugSegments=20,DebugThickness=0.25)
    DriverAxis=(X=-1,Y=0,Z=0)
    DriverItem=(Type=Bone,Name="thigh_l")
    ExecutePin=()
    FalloffRegionScaleFactors=(PositiveWidth=1,NegativeWidth=1,PositiveHeight=1,NegativeHeight=1)
    FalloffSize=0.3
    FlipHeightScaling=False
    FlipWidthScaling=False
    OptionalParentItem=(Type=Bone,Name="pelvis")
    RotationOffset=(X=90,Y=0,Z=10)
   )
   > &RigUnit_ModifyTransforms(
    ItemToModify="((Item=(Type=Bone,Name=\"dyn_underskirt_A_l_MSTR\"),Transform=(Rotation=(X=0,Y=-0.0175,Z=0,W=0.9998),Translation=(X=0,Y=0,Z=4),Scale3D=(X=1,Y=1,Z=1))),(Item=(Type=Bone,Name=\"dyn_sash_l_MSTR\"),Transform=(Rotation=(X=0,Y=-0.0131,Z=0,W=0.9999),Translation=(X=0, ...(truncated)"
    WeightMinimum=0
    Weight=@n22.OutputParam
   )
   > n23 = &RigUnit_SphericalPoseReader(
    ActiveRegionScaleFactors=(PositiveWidth=1,NegativeWidth=1,PositiveHeight=1,NegativeHeight=1)
    ActiveRegionSize=0.2
    Debug=(bDrawDebug=false,bDraw2D=False,bDrawLocalAxes=False,DebugScale=25,DebugSegments=20,DebugThickness=0.25)
    DriverAxis=(X=1,Y=0,Z=0)
    DriverItem=(Type=Bone,Name="thigh_r")
    FalloffRegionScaleFactors=(PositiveWidth=1,NegativeWidth=1,PositiveHeight=1,NegativeHeight=1)
    FalloffSize=0.3
    FlipHeightScaling=False
    FlipWidthScaling=False
    OptionalParentItem=(Type=Bone,Name="pelvis")
    RotationOffset=(X=90,Y=0,Z=-170)
   )
   > &RigUnit_ModifyTransforms(
    ItemToModify="((Item=(Type=Bone,Name=\"dyn_underskirt_A_r_MSTR\"),Transform=(Rotation=(X=0,Y=0,Z=0,W=1),Translation=(X=0,Y=0,Z=-4),Scale3D=(X=1,Y=1,Z=1))),(Item=(Type=Bone,Name=\"dyn_sash_r_MSTR\"),Transform=(Rotation=(X=0.1043,Y=-0.0694,Z=0.0073,W=0.9921),Translation=(X=0, ...(truncated)"
    WeightMinimum=0
    Weight=@n23.OutputParam
   )
   &RigUnit_ModifyTransforms(
    ExecutePin=()
    ItemToModify="((Item=(Type=Bone,Name=\"dyn_underskirt_C_l_MSTR\"),Transform=(Rotation=(X=0,Y=0,Z=0,W=1),Translation=(X=0,Y=0,Z=0),Scale3D=(X=1,Y=1,Z=1))),(Item=(Type=Bone,Name=\"dyn_underskirt_D_MSTR\"),Transform=(Rotation=(X=0,Y=0,Z=0,W=1),Translation=(X=0,Y=0,Z=0),Scale3D ...(truncated)"
    Weight=1
    WeightMinimum=0
   )
  ]}
]
hierarchy [
 root {loc=<0 0 0>, rot=<0 0 0 1>}
 pelvis {parent=root, loc=<0 -0.7071 93.1191>, rot=<0.0127 0.707 -0.0127 -0.707>}
 spine_01 {parent=pelvis, loc=<6.7494 -0.6388 0>, rot=<0 0 0.034 -0.9994>}
 spine_02 {parent=spine_01, loc=<7.2518 -0.035 0>, rot=<0 0 0.0238 -0.9997>}
 spine_03 {parent=spine_02, loc=<6.0104 0 0>, rot=<0 0 0.0149 -0.9999>}
 spine_04 {parent=spine_03, loc=<8.1024 0 0>, rot=<0 0 -0.0132 -0.9999>}
 spine_05 {parent=spine_04, loc=<10.0146 0 0>, rot=<0 0 -0.0015 -1>}
 neck_01 {parent=spine_05, loc=<11.3138 0.1529 0>, rot=<0 0 -0.2003 -0.9797>}
 neck_02 {parent=neck_01, loc=<4.5627 0 0>, rot=<0 0 0.0504 -0.9987>}
 head {parent=neck_02, loc=<5.3914 0 0>, rot=<0 0 0.115 -0.9934>}
 attach_cape {parent=spine_05, loc=<8.9676 -13.6268 -0.6524>, rot=<0.0618 -0.7045 -0.0622 -0.7042>}
 cape {parent=attach_cape, type=Null, loc=<0 -5.5931 0>, rot=<0 0 0 1>}
 attach_backpack {parent=spine_05, loc=<0.1405 -15.492 -0.0997>, rot=<-0.0092 -0.7071 0.0089 -0.707>}
 Backpack_BR {parent=attach_backpack, type=Null, loc=<0 -3.9737 0>, rot=<-0.0314 0 0 0.9995>}
 clavicle_l {parent=spine_05, loc=<4.3575 -0.1466 -2.3223>, rot=<0.0465 -0.7289 -0.009 -0.6829>}
 upperarm_l {parent=clavicle_l, loc=<15.5111 0 0>, rot=<0.0226 -0.3234 -0.0088 -0.9459>}
 lowerarm_l {parent=upperarm_l, loc=<26.6514 0 0>, rot=<0.0391 -0.0099 -0.2453 -0.9686>}
 hand_l {parent=lowerarm_l, loc=<22.3491 0 0>, rot=<-0.7081 -0.0507 0.0975 -0.6975>}
 thumb_01_l {parent=hand_l, loc=<1.3118 -0.3191 -2.459>, rot=<0.6546 -0.4042 0.0664 -0.6354>}
 thumb_02_l {parent=thumb_01_l, loc=<5.9336 -0.0448 0.0401>, rot=<-0.0037 0.0213 0.298 -0.9543>}
 thumb_03_l {parent=thumb_02_l, loc=<2.2223 0.3682 -0.0907>, rot=<0.0118 -0.013 0.0066 -0.9998>}
 index_metacarpal_l {parent=hand_l, loc=<1.6535 0.0183 -2.96>, rot=<0.0158 -0.0124 -0.0173 -0.9996>}
 index_01_l {parent=index_metacarpal_l, loc=<8.3008 0.0357 -0.0524>, rot=<-0.0044 0.0485 0.0924 -0.9945>}
 index_02_l {parent=index_01_l, loc=<3.3587 -0.3004 0.0452>, rot=<-0.002 0.0027 0.3225 -0.9466>}
 index_03_l {parent=index_02_l, loc=<2.5578 -0.123 -0.0252>, rot=<-0.002 0.0067 0.2558 -0.9667>}
 middle_metacarpal_l {parent=hand_l, loc=<2.1856 -0.1725 -0.5984>, rot=<0.006 0.0168 -0.0329 -0.9993>}
 middle_01_l {parent=middle_metacarpal_l, loc=<7.3275 -0.2447 0.0129>, rot=<-0.0208 0.0042 0.0858 -0.9961>}
 middle_02_l {parent=middle_01_l, loc=<4.0123 0.1401 0.1175>, rot=<-0.0218 0.0258 0.3512 -0.9357>}
 middle_03_l {parent=middle_02_l, loc=<2.6077 0.0534 0.1373>, rot=<0.0082 0.0188 0.1708 -0.9851>}
 ring_metacarpal_l {parent=hand_l, loc=<2.2074 -0.3625 1.3638>, rot=<-0.0284 0.0722 -0.0458 -0.9959>}
 ring_01_l {parent=ring_metacarpal_l, loc=<7.0295 -0.6197 -0.1446>, rot=<-0.0489 -0.0252 0.1892 -0.9804>}
 ring_02_l {parent=ring_01_l, loc=<3.6174 0.4647 -0.1027>, rot=<0 0 0.2806 -0.9598>}
 ring_03_l {parent=ring_02_l, loc=<2.4152 0.114 0.0199>, rot=<0 0 0.1189 -0.9929>}
 pinky_metacarpal_l {parent=hand_l, loc=<1.7616 -0.9753 3.3814>, rot=<-0.0397 0.0814 -0.069 -0.9935>}
 pinky_01_l {parent=pinky_metacarpal_l, loc=<7.1709 -0.4831 -0.1094>, rot=<-0.0654 -0.0035 0.2242 -0.9723>}
 pinky_02_l {parent=pinky_01_l, loc=<2.8274 0.0893 -0.1043>, rot=<0.0037 0.0011 0.2854 -0.9584>}
 pinky_03_l {parent=pinky_02_l, loc=<2.2046 0.0307 -0.0078>, rot=<-0.0002 0.0098 0.1445 -0.9895>}
 deform_wrist_dn_base_l {parent=hand_l, loc=<0 -2.2342 0>, rot=<0 0 0 -1>}
 deform_wrist_dn_l {parent=deform_wrist_dn_base_l, loc=<0 0 0>, rot=<0 0 0 1>}
 deform_wrist_up_base_l {parent=hand_l, loc=<0 2.2682 0>, rot=<0 0 0 -1>}
 deform_wrist_up_l {parent=deform_wrist_up_base_l, loc=<0 0 0>, rot=<0 0 0 1>}
 lowerarm_twist_02_l {parent=lowerarm_l, loc=<9.6199 0 0>, rot=<0 0 0 -1>}
 lowerarm_twist_01_l {parent=lowerarm_l, loc=<17.8549 0 0>, rot=<0 0 0 -1>}
 deform_elbow_in_base_l {parent=lowerarm_l, loc=<0.2112 5.259 0.8274>, rot=<0 0 0 -1>}
 deform_elbow_in_l {parent=deform_elbow_in_base_l, loc=<0 0 0>, rot=<0 0 0 1>}
 deform_elbow_out_base_l {parent=lowerarm_l, loc=<0 -4 0>, rot=<0 0 0 -1>}
 deform_elbow_out_l {parent=deform_elbow_out_base_l, loc=<0 0 0>, rot=<0 0 0 1>}
 dyn_lowerarm_sleeve_l_MSTR {parent=lowerarm_l, loc=<0 0 0>, rot=<0 0 0 -1>}
 dyn_lowerarm_l_sleeve_twist_01_MSTR {parent=dyn_lowerarm_sleeve_l_MSTR, loc=<0 0 0>, rot=<0 0 0 -1>}
 dyn_lowerarm_l_sleeve_twist_01 {parent=dyn_lowerarm_l_sleeve_twist_01_MSTR, loc=<0 0 0>, rot=<0 0 0 1>}
 dyn_lowerarm_l_sleeve_twist_01_end {parent=dyn_lowerarm_l_sleeve_twist_01, loc=<0.0003 0 -5.8435>, rot=<0 0 0 1>}
 dyn_lowerarm_l_sleeve_twist_02_MSTR {parent=dyn_lowerarm_sleeve_l_MSTR, loc=<9.6201 0 0.0001>, rot=<0 0 0 -1>}
 dyn_lowerarm_l_sleeve_twist_02 {parent=dyn_lowerarm_l_sleeve_twist_02_MSTR, loc=<0.0016 -0.0004 -0.0044>, rot=<0.0075 -0.0128 -0.0138 -0.9998>}
 dyn_lowerarm_l_sleeve_twist_02_end {parent=dyn_lowerarm_l_sleeve_twist_02, loc=<0.0047 -0.0014 -7.0652>, rot=<-0.0086 -0.0501 -0.042 -0.9978>}
 dyn_lowerarm_l_sleeve_twist_03_MSTR {parent=dyn_lowerarm_sleeve_l_MSTR, loc=<17.8552 0 0.0002>, rot=<0 0 0 -1>}
 dyn_lowerarm_l_sleeve_twist_03 {parent=dyn_lowerarm_l_sleeve_twist_03_MSTR, loc=<0.0011 -0.0004 -0.0045>, rot=<0.0077 -0.0139 -0.0129 -0.9998>}
 dyn_lowerarm_l_sleeve_twist_03_end {parent=dyn_lowerarm_l_sleeve_twist_03, loc=<0.0038 -0.001 -7.0653>, rot=<-0.0089 -0.0507 -0.0463 -0.9976>}
 upperarm_twist_01_l {parent=upperarm_l, loc=<0.0041 0 0>, rot=<0 0 0 -1>}
 deform_upperarm_bk_base_l {parent=upperarm_twist_01_l, loc=<1.0813 -7.2193 -1.9453>, rot=<0 0 0 1>}
 deform_upperarm_bk_l {parent=deform_upperarm_bk_base_l, loc=<0 0 0>, rot=<0 0 0 1>}
 deform_upperarm_fr_base_l {parent=upperarm_twist_01_l, loc=<-0.3691 7.0506 -2.765>, rot=<0 0 0 1>}
 deform_upperarm_fr_l {parent=deform_upperarm_fr_base_l, loc=<0 0 0>, rot=<0 0 0 1>}
 deform_upperarm_up_base_l {parent=upperarm_twist_01_l, loc=<-3.2216 -1.1391 3.3311>, rot=<0 0 0 1>}
 deform_upperarm_up_l {parent=deform_upperarm_up_base_l, loc=<0 0 0>, rot=<0 0 0 1>}
 upperarm_twist_02_l {parent=upperarm_l, loc=<13.7941 0 0>, rot=<0 0 0 -1>}
 clavicle_r {parent=spine_05, loc=<4.3575 -0.1466 2.3223>, rot=<0.7289 0.0465 -0.6829 0.009>}
 upperarm_r {parent=clavicle_r, loc=<-15.5111 0 0>, rot=<0.0226 -0.3234 -0.0088 -0.9459>}
 lowerarm_r {parent=upperarm_r, loc=<-26.6514 0 0>, rot=<0.0391 -0.0099 -0.2453 -0.9686>}
 hand_r {parent=lowerarm_r, loc=<-22.3491 0 0>, rot=<-0.7081 -0.0507 0.0975 -0.6975>}
 thumb_01_r {parent=hand_r, loc=<-1.3118 0.3191 2.459>, rot=<0.6546 -0.4042 0.0664 -0.6354>}
 thumb_02_r {parent=thumb_01_r, loc=<-5.9336 0.0448 -0.0401>, rot=<-0.0037 0.0213 0.298 -0.9543>}
 thumb_03_r {parent=thumb_02_r, loc=<-2.2223 -0.3682 0.0907>, rot=<0.0118 -0.013 0.0066 -0.9998>}
 index_metacarpal_r {parent=hand_r, loc=<-1.6535 -0.0183 2.96>, rot=<0.0158 -0.0124 -0.0173 -0.9996>}
 index_01_r {parent=index_metacarpal_r, loc=<-8.3008 -0.0357 0.0524>, rot=<-0.0044 0.0485 0.0924 -0.9945>}
 index_02_r {parent=index_01_r, loc=<-3.3587 0.3004 -0.0452>, rot=<-0.002 0.0027 0.3225 -0.9466>}
 index_03_r {parent=index_02_r, loc=<-2.5578 0.123 0.0252>, rot=<-0.002 0.0067 0.2558 -0.9667>}
 middle_metacarpal_r {parent=hand_r, loc=<-2.1856 0.1725 0.5984>, rot=<0.006 0.0168 -0.0329 -0.9993>}
 middle_01_r {parent=middle_metacarpal_r, loc=<-7.3275 0.2447 -0.0129>, rot=<-0.0208 0.0042 0.0858 -0.9961>}
 middle_02_r {parent=middle_01_r, loc=<-4.0123 -0.1401 -0.1175>, rot=<-0.0218 0.0258 0.3512 -0.9357>}
 middle_03_r {parent=middle_02_r, loc=<-2.6077 -0.0534 -0.1373>, rot=<0.0082 0.0188 0.1708 -0.9851>}
 ring_metacarpal_r {parent=hand_r, loc=<-2.2074 0.3625 -1.3638>, rot=<-0.0284 0.0722 -0.0458 -0.9959>}
 ring_01_r {parent=ring_metacarpal_r, loc=<-7.0295 0.6197 0.1446>, rot=<-0.0489 -0.0252 0.1892 -0.9804>}
 ring_02_r {parent=ring_01_r, loc=<-3.6174 -0.4647 0.1027>, rot=<0 0 0.2806 -0.9598>}
 ring_03_r {parent=ring_02_r, loc=<-2.4152 -0.114 -0.0199>, rot=<0 0 0.1189 -0.9929>}
 pinky_metacarpal_r {parent=hand_r, loc=<-1.7616 0.9753 -3.3814>, rot=<-0.0397 0.0814 -0.069 -0.9935>}
 pinky_01_r {parent=pinky_metacarpal_r, loc=<-7.1709 0.4831 0.1094>, rot=<-0.0654 -0.0035 0.2242 -0.9723>}
 pinky_02_r {parent=pinky_01_r, loc=<-2.8274 -0.0893 0.1043>, rot=<0.0037 0.0011 0.2854 -0.9584>}
 pinky_03_r {parent=pinky_02_r, loc=<-2.2046 -0.0307 0.0078>, rot=<-0.0002 0.0098 0.1445 -0.9895>}
 deform_wrist_dn_base_r {parent=hand_r, loc=<-0.0004 2.2345 -0.0001>, rot=<0 0 0 -1>}
 deform_wrist_dn_r {parent=deform_wrist_dn_base_r, loc=<0 0 0>, rot=<0 0 0 1>}
 deform_wrist_up_base_r {parent=hand_r, loc=<0.0003 -2.2684 0.0001>, rot=<0 0 0 -1>}
 deform_wrist_up_r {parent=deform_wrist_up_base_r, loc=<0 0 0>, rot=<0 0 0 1>}
 lowerarm_twist_02_r {parent=lowerarm_r, loc=<-9.6199 0 0>, rot=<0 0 0 -1>}
 lowerarm_twist_01_r {parent=lowerarm_r, loc=<-17.8549 0 0>, rot=<0 0 0 -1>}
 deform_elbow_in_base_r {parent=lowerarm_r, loc=<-0.2111 -5.2591 -0.8275>, rot=<0 0 0 -1>}
 deform_elbow_in_r {parent=deform_elbow_in_base_r, loc=<-0.0001 0 0.0002>, rot=<0 0 0 1>}
 deform_elbow_out_base_r {parent=lowerarm_r, loc=<0 4 -0.0001>, rot=<0 0 0 -1>}
 deform_elbow_out_r {parent=deform_elbow_out_base_r, loc=<0 0 0>, rot=<0 0 0 1>}
 dyn_lowerarm_sleeve_r_MSTR {parent=lowerarm_r, loc=<0 0 0>, rot=<0 0 0 -1>}
 dyn_lowerarm_r_sleeve_twist_03_MSTR {parent=dyn_lowerarm_sleeve_r_MSTR, loc=<-17.8547 -0.0001 -0.0001>, rot=<0 0 0 -1>}
 dyn_lowerarm_r_sleeve_twist_03 {parent=dyn_lowerarm_r_sleeve_twist_03_MSTR, loc=<-0.0002 0.0011 0.0015>, rot=<-0.0132 -0.0139 0.0144 -0.9997>}
 dyn_lowerarm_r_sleeve_twist_03_end {parent=dyn_lowerarm_r_sleeve_twist_03, loc=<-0.017 -0.0101 7.0786>, rot=<0.0043 -0.048 0.0188 -0.9987>}
 dyn_lowerarm_r_sleeve_twist_02_MSTR {parent=dyn_lowerarm_sleeve_r_MSTR, loc=<-9.6198 -0.0001 -0.0001>, rot=<0 0 0 -1>}
 dyn_lowerarm_r_sleeve_twist_02 {parent=dyn_lowerarm_r_sleeve_twist_02_MSTR, loc=<-0.0001 0.0011 0.0019>, rot=<-0.0132 -0.0139 0.0143 -0.9997>}
 dyn_lowerarm_r_sleeve_twist_02_end {parent=dyn_lowerarm_r_sleeve_twist_02, loc=<-0.0192 -0.0185 7.084>, rot=<0.0053 -0.015 -0.0093 -0.9998>}
 dyn_lowerarm_r_sleeve_twist_01_MSTR {parent=dyn_lowerarm_sleeve_r_MSTR, loc=<-0.0001 0 0.0002>, rot=<0 0 0 -1>}
 dyn_lowerarm_r_sleeve_twist_01 {parent=dyn_lowerarm_r_sleeve_twist_01_MSTR, loc=<0 0 0>, rot=<0 0 0 1>}
 dyn_lowerarm_r_sleeve_twist_01_end {parent=dyn_lowerarm_r_sleeve_twist_01, loc=<0.0007 -0.0001 5.8435>, rot=<0 0 0 1>}
 upperarm_twist_01_r {parent=upperarm_r, loc=<-0.0041 0 0>, rot=<0 0 0 -1>}
 deform_upperarm_bk_base_r {parent=upperarm_twist_01_r, loc=<-1.0814 7.2192 1.9455>, rot=<0 0 0 1>}
 deform_upperarm_bk_r {parent=deform_upperarm_bk_base_r, loc=<0 0 0>, rot=<0 0 0 1>}
 deform_upperarm_fr_base_r {parent=upperarm_twist_01_r, loc=<0.3689 -7.0506 2.7652>, rot=<0 0 0 1>}
 deform_upperarm_fr_r {parent=deform_upperarm_fr_base_r, loc=<0 0 0>, rot=<0 0 0 1>}
 deform_upperarm_up_base_r {parent=upperarm_twist_01_r, loc=<3.2214 1.1391 -3.3308>, rot=<0 0 0 1>}
 deform_upperarm_up_r {parent=deform_upperarm_up_base_r, loc=<0 0 0>, rot=<0 0 0 -1>}
 upperarm_twist_02_r {parent=upperarm_r, loc=<-13.7941 0 0>, rot=<0 0 0 -1>}
 deform_pec_aim_base_l {parent=spine_04, loc=<7.8732 12.6388 -2.4123>, rot=<0.0102 -0.4602 0.3979 -0.7936>}
 deform_pec_aim_l {parent=deform_pec_aim_base_l, loc=<0 0 0>, rot=<0 0 0 1>}
 deform_pec_base_l {parent=deform_pec_aim_l, loc=<10.766 0 0>, rot=<0 0 0 1>}
 deform_pec_l {parent=deform_pec_base_l, loc=<0.055 0 0>, rot=<0 0 0 1>}
 deform_pec_aim_base_r {parent=spine_04, loc=<7.8727 12.6388 2.4123>, rot=<-0.4602 -0.0102 0.7936 0.3979>}
 deform_pec_aim_r {parent=deform_pec_aim_base_r, loc=<0 0 0>, rot=<0 0 0 1>}
 deform_pec_base_r {parent=deform_pec_aim_r, loc=<-10.7662 -0.0006 -0.0007>, rot=<0 0 0 1>}
 deform_pec_r {parent=deform_pec_base_r, loc=<0 0 0>, rot=<0 0 0 1>}
 dyn_cloak_A_01_l_MSTR {parent=spine_03, loc=<1.9652 10.7388 -15.1865>, rot=<0.0144 -0.9382 -0.3346 -0.0869>}
 dyn_cloak_A_01_l {parent=dyn_cloak_A_01_l_MSTR, loc=<0 0 0>, rot=<0 0 0 1>}
 dyn_cloak_A_02_l {parent=dyn_cloak_A_01_l, loc=<10.5949 0 0>, rot=<-0.0367 -0.0405 0.0068 -0.9985>}
 dyn_cloak_A_03_l {parent=dyn_cloak_A_02_l, loc=<9.1902 0 0>, rot=<0 0 0 1>}
 dyn_cloak_A_04_l {parent=dyn_cloak_A_03_l, loc=<13.2421 0 0>, rot=<0 0 0 1>}
 dyn_cloak_A_05_l {parent=dyn_cloak_A_04_l, loc=<16.8356 0 0>, rot=<0 0 0 1>}
 dyn_cloak_A_06_l {parent=dyn_cloak_A_05_l, loc=<15.6255 0 0>, rot=<0 0 0 1>}
 dyn_cloak_A_07_l {parent=dyn_cloak_A_06_l, loc=<15.6325 0 0>, rot=<0 0 0 1>}
 dyn_cloak_B_01_l_MSTR {parent=spine_03, loc=<2.7108 2.0414 -19.4052>, rot=<0.0583 -0.992 0.103 -0.043>}
 dyn_cloak_B_01_l {parent=dyn_cloak_B_01_l_MSTR, loc=<0 0 0>, rot=<0 0 0 1>}
 dyn_cloak_B_02_l {parent=dyn_cloak_B_01_l, loc=<9.8925 0 0>, rot=<-0.004 0.0324 0.0004 -0.9995>}
 dyn_cloak_B_03_l {parent=dyn_cloak_B_02_l, loc=<8.8684 0 0>, rot=<0 0 0 1>}
 dyn_cloak_B_04_l {parent=dyn_cloak_B_03_l, loc=<13.7717 0 0>, rot=<0 0 0 1>}
 dyn_cloak_B_05_l {parent=dyn_cloak_B_04_l, loc=<16.7509 0 0>, rot=<0 0 0 1>}
 dyn_cloak_B_06_l {parent=dyn_cloak_B_05_l, loc=<15.8096 0 0>, rot=<0 0 0 1>}
 dyn_cloak_B_07_l {parent=dyn_cloak_B_06_l, loc=<15.4275 0 0>, rot=<0 0 0 1>}
 dyn_cloak_C_01_l_MSTR {parent=spine_03, loc=<4.503 -11.638 -14.3796>, rot=<0.0651 -0.9213 0.3793 -0.0562>}
 dyn_cloak_C_01_l {parent=dyn_cloak_C_01_l_MSTR, loc=<0 0 0>, rot=<0 0 0 1>}
 dyn_cloak_C_02_l {parent=dyn_cloak_C_01_l, loc=<18.6829 0 0>, rot=<0 0 0 1>}
 dyn_cloak_C_03_l {parent=dyn_cloak_C_02_l, loc=<13.7036 0 0>, rot=<0 0 0 1>}
 dyn_cloak_C_04_l {parent=dyn_cloak_C_03_l, loc=<16.1036 0 0>, rot=<0 0 0 1>}
 dyn_cloak_C_05_l {parent=dyn_cloak_C_04_l, loc=<15.9967 0 0>, rot=<0 0 0 1>}
 dyn_cloak_C_06_l {parent=dyn_cloak_C_05_l, loc=<15.8899 0 0>, rot=<0 0 0 1>}
 dyn_cloak_A_01_r_MSTR {parent=spine_03, loc=<2.6406 10.6983 15.0498>, rot=<0.0407 0.3716 0.9271 -0.0265>}
 dyn_cloak_A_01_r {parent=dyn_cloak_A_01_r_MSTR, loc=<0 0 0>, rot=<0 0 0 1>}
 dyn_cloak_A_02_r {parent=dyn_cloak_A_01_r, loc=<11.8653 0 0>, rot=<0 0 0 1>}
 dyn_cloak_A_03_r {parent=dyn_cloak_A_02_r, loc=<9.0099 0.0001 0>, rot=<0 0 0 1>}
 dyn_cloak_A_04_r {parent=dyn_cloak_A_03_r, loc=<12.9352 0.0001 0.0001>, rot=<0 0 0 -1>}
 dyn_cloak_A_05_r {parent=dyn_cloak_A_04_r, loc=<16.0533 0 0>, rot=<0 0 0 1>}
 dyn_cloak_A_06_r {parent=dyn_cloak_A_05_r, loc=<16.0304 -0.0001 0>, rot=<0 0 0 1>}
 dyn_cloak_A_07_r {parent=dyn_cloak_A_06_r, loc=<16.4435 0 0>, rot=<0 0 0 1>}
 dyn_cloak_B_01_r_MSTR {parent=spine_03, loc=<3.2848 2.122 19.3268>, rot=<0.0536 -0.1042 0.9913 -0.0595>}
 dyn_cloak_B_01_r {parent=dyn_cloak_B_01_r_MSTR, loc=<0 0 0>, rot=<0 0 0 1>}
 dyn_cloak_B_02_r {parent=dyn_cloak_B_01_r, loc=<10.9429 0 0>, rot=<0.002 0.0195 -0.0002 -0.9998>}
 dyn_cloak_B_03_r {parent=dyn_cloak_B_02_r, loc=<8.9934 0 0>, rot=<0 0 0 -1>}
 dyn_cloak_B_04_r {parent=dyn_cloak_B_03_r, loc=<13.2353 0 0>, rot=<0 0 0 -1>}
 dyn_cloak_B_05_r {parent=dyn_cloak_B_04_r, loc=<15.7945 0 0>, rot=<0 0 0 1>}
 dyn_cloak_B_06_r {parent=dyn_cloak_B_05_r, loc=<16.3352 0 0.0001>, rot=<0 0 0 1>}
 dyn_cloak_B_07_r {parent=dyn_cloak_B_06_r, loc=<16.4826 0 0>, rot=<0 0 0 1>}
 dyn_cloak_C_01_r_MSTR {parent=spine_03, loc=<4.5082 -12.0256 14.608>, rot=<0.0543 -0.3735 0.9238 -0.0644>}
 dyn_cloak_C_01_r {parent=dyn_cloak_C_01_r_MSTR, loc=<0 0 0>, rot=<0 0 0 1>}
 dyn_cloak_C_02_r {parent=dyn_cloak_C_01_r, loc=<18.6698 0 0>, rot=<0 0 0 -1>}
 dyn_cloak_C_03_r {parent=dyn_cloak_C_02_r, loc=<13.7859 0 0>, rot=<0 0 0 1>}
 dyn_cloak_C_04_r {parent=dyn_cloak_C_03_r, loc=<15.8277 0 0>, rot=<0 0 0 1>}
 dyn_cloak_C_05_r {parent=dyn_cloak_C_04_r, loc=<16.105 0 0>, rot=<0 0 0 1>}
 dyn_cloak_C_06_r {parent=dyn_cloak_C_05_r, loc=<16.654 0 0>, rot=<0 0 0 1>}
 dyn_cloak_D_08_MSTR {parent=spine_03, loc=<4.7712 -13.4367 -0.0149>, rot=<0.0573 -0.7068 0.7027 -0.0572>}
 dyn_cloak_D_01 {parent=dyn_cloak_D_08_MSTR, loc=<0 0 0>, rot=<0 0 0 1>}
 dyn_cloak_D_02 {parent=dyn_cloak_D_01, loc=<18.5885 0 0>, rot=<-0.0366 0.0001 0.0019 -0.9993>}
 dyn_cloak_D_03 {parent=dyn_cloak_D_02, loc=<13.8649 0 0>, rot=<0.0346 -0.0001 -0.0018 -0.9994>}
 dyn_cloak_D_04 {parent=dyn_cloak_D_03, loc=<15.983 0 0>, rot=<-0.0011 0 0.0001 -1>}
 dyn_cloak_D_05 {parent=dyn_cloak_D_04, loc=<16.0587 0 0>, rot=<-0.0013 0 0.0001 -1>}
 dyn_cloak_D_06 {parent=dyn_cloak_D_05, loc=<16.1544 0 0>, rot=<0 0 0 1>}
 thigh_l {parent=pelvis, loc=<-2.5693 -0.0109 -7.7501>, rot=<0.0012 0.0572 0.0155 -0.9982>}
 calf_l {parent=thigh_l, loc=<-39.0484 0 0>, rot=<0.0006 -0.0161 -0.0391 -0.9991>}
 foot_l {parent=calf_l, loc=<-42.561 0.1077 0.4292>, rot=<0.0418 -0.0455 0.0343 -0.9975>}
 ball_l {parent=foot_l, loc=<-6.1599 13.8264 0.6806>, rot=<0.0001 -0.0001 -0.7063 -0.7079>}
 calf_twist_02_l {parent=calf_l, loc=<-18.8064 -0.0967 0.0397>, rot=<0.0237 0.0095 -0.0035 -0.9997>}
 calf_twist_01_l {parent=calf_l, loc=<-35.813 -0.0967 0.0397>, rot=<0.0237 0.0095 -0.0035 -0.9997>}
 deform_knee_in_base_l {parent=calf_l, loc=<1.298 -5.2957 -0.0496>, rot=<0 0 0 -1>}
 deform_knee_in_l {parent=deform_knee_in_base_l, loc=<0 0 0>, rot=<0 0 0 1>}
 deform_knee_out_base_l {parent=calf_l, loc=<0 4 0>, rot=<0 0 0 -1>}
 deform_knee_out_l {parent=deform_knee_out_base_l, loc=<0 0 0>, rot=<0 0 0 1>}
 thigh_twist_01_l {parent=thigh_l, loc=<-21.6303 0 0>, rot=<0.0013 0 0 -1>}
 dyn_underskirt_A_l_MSTR {parent=thigh_twist_01_l, loc=<23.0269 10.391 0.5267>, rot=<-0.0359 -0.842 -0.5382 0.0087>}
 dyn_sash_l_MSTR {parent=dyn_underskirt_A_l_MSTR, loc=<-2.7764 -0.2177 1.2349>, rot=<-0.0009 0.0141 -0.0042 -0.9999>}
 dyn_sash_l_01 {parent=dyn_sash_l_MSTR, loc=<0 0 0>, rot=<0 0 0 1>}
 dyn_sash_l_02 {parent=dyn_sash_l_01, loc=<17.7979 0 0>, rot=<0.0446 0.013 -0.0005 -0.9989>}
 dyn_sash_l_03 {parent=dyn_sash_l_02, loc=<15.8462 0 0>, rot=<0 0 0 1>}
 dyn_underskirt_A_l_01 {parent=dyn_underskirt_A_l_MSTR, loc=<0 0 0>, rot=<0 0 0 1>}
 dyn_underskirt_A_l_02 {parent=dyn_underskirt_A_l_01, loc=<15.0432 0 0>, rot=<0.0446 -0.005 0.002 -0.999>}
 dyn_underskirt_A_l_03 {parent=dyn_underskirt_A_l_02, loc=<15.8462 0 0>, rot=<0 0 0 1>}
 deform_groin_base_l {parent=thigh_l, loc=<-2.1908 8.1142 -1.7442>, rot=<0 0 0 -1>}
 deform_groin_l {parent=deform_groin_base_l, loc=<0 0 0>, rot=<0 0 0 1>}
 CR_thigh_l_notwist {parent=thigh_l, loc=<0 0 0>, rot=<0 0 0 1>}
 thigh_r {parent=pelvis, loc=<-2.5693 -0.0109 7.7501>, rot=<0.0572 -0.0012 0.9982 0.0155>}
 calf_r {parent=thigh_r, loc=<39.0484 0 0>, rot=<0.0006 -0.0161 -0.0391 -0.9991>}
 foot_r {parent=calf_r, loc=<42.561 -0.1077 -0.4292>, rot=<0.0418 -0.0455 0.0343 -0.9975>}
 ball_r {parent=foot_r, loc=<6.1599 -13.8264 -0.6806>, rot=<0.0001 -0.0001 -0.7063 -0.7079>}
 calf_twist_02_r {parent=calf_r, loc=<18.8064 0.0967 -0.0397>, rot=<0.0237 0.0095 -0.0035 -0.9997>}
 calf_twist_01_r {parent=calf_r, loc=<35.813 0.0967 -0.0397>, rot=<0.0237 0.0095 -0.0035 -0.9997>}
 deform_knee_in_base_r {parent=calf_r, loc=<-1.298 5.2957 0.0496>, rot=<0 0 0 -1>}
 deform_knee_in_r {parent=deform_knee_in_base_r, loc=<0 0 0>, rot=<0 0 0 1>}
 deform_knee_out_base_r {parent=calf_r, loc=<0 -4 0>, rot=<0 0 0 -1>}
 deform_knee_out_r {parent=deform_knee_out_base_r, loc=<0 0 0>, rot=<0 0 0 1>}
 thigh_twist_01_r {parent=thigh_r, loc=<21.6303 0 0>, rot=<0.0013 0 0 -1>}
 dyn_underskirt_A_r_MSTR {parent=thigh_twist_01_r, loc=<-23.0269 -10.391 -0.5267>, rot=<-0.0365 -0.7919 -0.6095 0.0055>}
 dyn_underskirt_A_r_01 {parent=dyn_underskirt_A_r_MSTR, loc=<0 0 0>, rot=<0 0 0 1>}
 dyn_underskirt_A_r_02 {parent=dyn_underskirt_A_r_01, loc=<-15.0432 0 0>, rot=<-0.0002 -0.0022 -0.0175 -0.9998>}
 dyn_underskirt_A_r_03 {parent=dyn_underskirt_A_r_02, loc=<-15.8348 0.595 0.0961>, rot=<0 0 0 1>}
 dyn_sash_r_MSTR {parent=dyn_underskirt_A_r_MSTR, loc=<2.7764 0.2177 -1.2349>, rot=<-0.0009 0.0141 -0.0042 -0.9999>}
 dyn_sash_r_01 {parent=dyn_sash_r_MSTR, loc=<0 0 0>, rot=<0 0 0 1>}
 dyn_sash_r_02 {parent=dyn_sash_r_01, loc=<-17.7979 0 0>, rot=<-0.0002 -0.0022 -0.0175 -0.9998>}
 dyn_sash_r_03 {parent=dyn_sash_r_02, loc=<-15.8345 0.6309 0.5297>, rot=<0 0 0 -1>}
 deform_groin_base_r {parent=thigh_r, loc=<2.1908 -8.1142 1.7442>, rot=<0 0 0 -1>}
 deform_groin_r {parent=deform_groin_base_r, loc=<0 0 0>, rot=<0 0 0 1>}
 deform_glute_base_l {parent=pelvis, loc=<-8.8064 -8.8222 -8.4287>, rot=<0.0275 0.0501 -0.4472 -0.8926>}
 deform_glute_l {parent=deform_glute_base_l, loc=<0 0 0>, rot=<0 0 0 1>}
 deform_glute_base_r {parent=pelvis, loc=<-8.8062 -8.8222 8.4287>, rot=<0.0501 -0.0275 0.8926 -0.4472>}
 deform_glute_r {parent=deform_glute_base_r, loc=<0 0 0>, rot=<0 0 0 1>}
 deform_canister_03_r {parent=pelvis, loc=<8.2474 8.1359 12.1045>, rot=<0.6618 0.2814 0.6527 0.2385>}
 deform_canister_04_r {parent=pelvis, loc=<8.3345 7.1714 12.7421>, rot=<0.6689 0.2583 0.6628 0.2158>}
 deform_pouch_r {parent=pelvis, loc=<9.1556 4.117 14.1929>, rot=<0.7004 0.095 0.704 0.0697>}
 deform_canister_01_l {parent=pelvis, loc=<8.2789 3.9933 -14.4252>, rot=<-0.0031 -0.6949 -0.0145 -0.719>}
 deform_canister_02_l {parent=pelvis, loc=<8.306 2.6811 -14.6442>, rot=<-0.0074 -0.6947 -0.0065 -0.7192>}
 deform_canister_03_l {parent=pelvis, loc=<8.3063 1.44 -14.7213>, rot=<-0.0014 -0.6947 -0.0128 -0.7192>}
 deform_canister_04_l {parent=pelvis, loc=<8.3484 0.2438 -14.6516>, rot=<0.0099 -0.6948 -0.0019 -0.7191>}
 deform_pouch_l {parent=pelvis, loc=<9.4512 -2.8006 -13.9927>, rot=<0.1252 -0.7105 0.1042 -0.6846>}
 dyn_underskirt_B_l_MSTR {parent=pelvis, loc=<-1.3172 1.0316 -15.7361>, rot=<-0.0191 -0.9928 -0.0091 -0.1183>}
 dyn_underskirt_B_l_01 {parent=dyn_underskirt_B_l_MSTR, loc=<0 0 0>, rot=<0 0 0 1>}
 dyn_underskirt_B_l_02 {parent=dyn_underskirt_B_l_01, loc=<9.2603 0 0>, rot=<-0.026 -0.0329 0.0054 -0.9991>}
 dyn_underskirt_B_l_03 {parent=dyn_underskirt_B_l_02, loc=<11.3166 0 0>, rot=<0 0 0 1>}
 dyn_underskirt_C_l_MSTR {parent=pelvis, loc=<-1.3858 -9.3222 -10.174>, rot=<0.0126 -0.8123 0.5802 -0.0585>}
 dyn_underskirt_C_l_01 {parent=dyn_underskirt_C_l_MSTR, loc=<0 0 0>, rot=<0 0 0 1>}
 dyn_underskirt_C_l_02 {parent=dyn_underskirt_C_l_01, loc=<9.3467 0 0>, rot=<-0.0001 -0.0465 0.0025 -0.9989>}
 dyn_underskirt_C_l_03 {parent=dyn_underskirt_C_l_02, loc=<11.2042 0 0>, rot=<0 0 0 1>}
 dyn_underskirt_B_r_MSTR {parent=pelvis, loc=<-1.3172 1.0316 15.7361>, rot=<0.9928 -0.0191 -0.1183 0.0091>}
 dyn_underskirt_B_r_01 {parent=dyn_underskirt_B_r_MSTR, loc=<0 0 0>, rot=<0 0 0 1>}
 dyn_underskirt_B_r_02 {parent=dyn_underskirt_B_r_01, loc=<-9.2602 0 0>, rot=<-0.026 -0.0329 0.0054 -0.9991>}
 dyn_underskirt_B_r_03 {parent=dyn_underskirt_B_r_02, loc=<-11.3166 0 0>, rot=<0 0 0 1>}
 dyn_underskirt_C_r_MSTR {parent=pelvis, loc=<-1.3452 -9.6708 10.2926>, rot=<0.8123 0.0126 -0.0585 -0.5802>}
 dyn_underskirt_C_r_01 {parent=dyn_underskirt_C_r_MSTR, loc=<-0.0001 0.0001 0>, rot=<0 0 0 -1>}
 dyn_underskirt_C_r_02 {parent=dyn_underskirt_C_r_01, loc=<-9.3467 0 -0.0001>, rot=<-0.0001 -0.0465 0.0025 -0.9989>}
 dyn_underskirt_C_r_03 {parent=dyn_underskirt_C_r_02, loc=<-11.2042 0 0>, rot=<0 0 0 1>}
 dyn_underskirt_D_MSTR {parent=pelvis, loc=<-1.6791 -11.3738 0>, rot=<0.0185 -0.7069 0.7069 -0.0185>}
 dyn_underskirt_D_01 {parent=dyn_underskirt_D_MSTR, loc=<0 0 0>, rot=<0 0 0 1>}
 dyn_underskirt_D_02 {parent=dyn_underskirt_D_01, loc=<9.1391 -0.0015 -0.0062>, rot=<0.0283 -0.009 0.0019 -0.9996>}
 dyn_underskirt_D_03 {parent=dyn_underskirt_D_02, loc=<11.2379 0 0>, rot=<0 0 0 1>}
 dyn_lightsaber_MSTR {parent=pelvis, loc=<7.9735 8.4618 -15.951>, rot=<-0.0365 -0.957 -0.2788 -0.0713>}
 dyn_lightsaber {parent=dyn_lightsaber_MSTR, loc=<0 0 0>, rot=<0 0 0 1>}
 deform_canister_01_r {parent=pelvis, loc=<8.1338 9.9277 10.3936>, rot=<0.6512 0.3115 0.6379 0.2682>}
 deform_canister_02_r {parent=pelvis, loc=<8.2038 9.0239 11.3488>, rot=<0.6557 0.3111 0.6383 0.2566>}
 dyn_underskirt_front_MSTR {parent=pelvis, loc=<-0.7554 12.0558 -0.4233>, rot=<-0.0332 -0.7063 -0.7063 -0.0332>}
 dyn_underskirt_front_01 {parent=dyn_underskirt_front_MSTR, loc=<0 0 0>, rot=<0 0 0 1>}
 dyn_underskirt_front_02 {parent=dyn_underskirt_front_01, loc=<15.095 0 0.0785>, rot=<0 -0.0204 0 -0.9998>}
 dyn_underskirt_front_03 {parent=dyn_underskirt_front_02, loc=<15.8139 0 0>, rot=<0 0 0 1>}
 CR_thigh_r {parent=pelvis, loc=<-2.5693 -0.0109 7.7501>, rot=<-0.0004 -0.0572 0 0.9984>}
 CR_thigh_no_twist_r {parent=CR_thigh_r, loc=<0 0 0>, rot=<0 0 0 1>}
 dyn_skirt_ground_collision {parent=root, loc=<0 0 0>, rot=<0 0 0 1>}
 thigh_l_up {type=Curve, loc=<0 0 0>, rot=<0 0 0 1>}
 thigh_r_up {type=Curve, loc=<0 0 0>, rot=<0 0 0 1>}
]
variables [
 Disable Floor Slope {type=bool, default=False}
 Push Down Weight {type=double, default=0}
]
