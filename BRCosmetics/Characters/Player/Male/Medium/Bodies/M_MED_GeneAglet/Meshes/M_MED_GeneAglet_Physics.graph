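# Physics asset for the GeneAglet medium male player body cosmetic, supporting secondary motion of its cloak.
# Capsules cover the lower body, spine, cloak, sash, underskirt and lightsaber bones, plus a large box for skirt ground collision.
# Constraints chain seven cloak strands from tiny MSTR anchors, generally allowing more swing lower down.
# Angular-free constraints link the A and B strand tips on each side; sash, underskirt and lightsaber bodies have no constraints.
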PhysicsAsset /BRCosmetics/Characters/Player/Male/Medium/Bodies/M_MED_GeneAglet/Meshes/M_MED_GeneAglet_Physics
bodies {
 calf_l [{shapeType=Capsule, length=48.6, radius=8, shapeName=calf_l_capsule, center=<-22.9991 0.0005 0.3637>, rotation=<rot 87.5789 0.0035 0.0073>}]
 calf_r [{
   shapeType=Capsule
   length=48.6
   radius=8
   shapeName=calf_r_capsule
   center=<23.0719 0.021 -1.7037>
   rotation=<rot -87.5789 -179.8496 179.9927>
  }]
 dyn_cloak_A_01_l [{
   shapeType=Capsule
   length=12.1381
   radius=6.1451
   shapeName=dyn_cloak_A_01_l_capsule
   center=<14.5609 1.6783 4.7124>
   rotation=<rot 84.7339 -0.0001 -0.0002>
  }]
 dyn_cloak_A_01_l_MSTR [{shapeType=Capsule, length=0.505, radius=0.505, shapeName=dyn_cloak_A_01_l_MSTR_capsule}]
 dyn_cloak_A_01_r [{shapeType=Capsule, length=13.6656, radius=4.6175, shapeName=dyn_cloak_A_01_r_capsule, center=<14.1671 -0.5 2>, rotation=<rot 90 0 0>}]
 dyn_cloak_A_01_r_MSTR [{shapeType=Capsule, length=0.505, radius=0.505, shapeName=dyn_cloak_A_01_r_MSTR_capsule, rotation=<rot 0 -90 0>}]
 dyn_cloak_A_02_l [{shapeType=Capsule, length=13.3399, radius=6.6353, shapeName=dyn_cloak_A_02_l_capsule, center=<6.8305 1.482 5.5>, rotation=<rot 90 0 0>}]
 dyn_cloak_A_02_r [{shapeType=Capsule, length=9.6818, radius=4.8279, shapeName=dyn_cloak_A_02_r_capsule, center=<6.8305 -0.5 2>, rotation=<rot 90 0 0>}]
 dyn_cloak_A_03_l [{shapeType=Capsule, length=8.5086, radius=6.5662, shapeName=dyn_cloak_A_03_l_capsule, center=<7.4862 1.6741 5.5>, rotation=<rot 90 0 0>}]
 dyn_cloak_A_03_r [{shapeType=Capsule, length=9.9582, radius=5.1166, shapeName=dyn_cloak_A_03_r_capsule, center=<7.4862 -0.5 1.8>, rotation=<rot 90 0 0>}]
 dyn_cloak_A_04_l [{shapeType=Capsule, length=13.2218, radius=6.5593, shapeName=dyn_cloak_A_04_l_capsule, center=<9.3296 1.4703 5.5>, rotation=<rot 90 0 0>}]
 dyn_cloak_A_04_r [{shapeType=Capsule, length=14.1665, radius=6, shapeName=dyn_cloak_A_04_r_capsule, center=<8.6935 0 1.8>, rotation=<rot 90 0 0>}]
 dyn_cloak_A_05_l [{
   shapeType=Capsule
   length=10.7219
   radius=6.4397
   shapeName=dyn_cloak_A_05_l_capsule
   center=<9.108 0 2.8>
   rotation=<rot 88.0978 -0.0001 -0.0001>
  }]
 dyn_cloak_A_05_r [{shapeType=Capsule, length=11.7334, radius=6.4397, shapeName=dyn_cloak_A_05_r_capsule, center=<8.4464 0 1.0363>, rotation=<rot 90 0 0>}]
 dyn_cloak_A_06_l [{
   shapeType=Capsule
   length=11.7076
   radius=6.9397
   shapeName=dyn_cloak_A_06_l_capsule
   center=<9.0131 0 2.3>
   rotation=<rot 88.3809 -0.0001 -0.0001>
  }]
 dyn_cloak_A_06_r [{shapeType=Capsule, length=12.4692, radius=6.9397, shapeName=dyn_cloak_A_06_r_capsule, center=<8.3988 0 0.9459>, rotation=<rot 90 0 0>}]
 dyn_cloak_A_07_l [{shapeType=Capsule, length=16.2754, radius=8.7668, shapeName=dyn_cloak_A_07_l_capsule, center=<8.384 -3.7 1.6>, rotation=<rot 0 0 90>}]
 dyn_cloak_A_07_r [{shapeType=Capsule, length=15.0297, radius=8.7668, shapeName=dyn_cloak_A_07_r_capsule, center=<8.2333 -2.2918 1>, rotation=<rot 0 0 90>}]
 dyn_cloak_B_01_l [{
   shapeType=Capsule
   length=14.7914
   radius=6.2473
   shapeName=dyn_cloak_B_01_l_capsule
   center=<9.9245 0.9048 4.6>
   rotation=<rot -0.0001 0.0002 112.3075>
  }]
 dyn_cloak_B_01_l_MSTR [{shapeType=Capsule, length=0.505, radius=0.505, shapeName=dyn_cloak_B_01_l_MSTR_capsule, rotation=<rot 0 -90 0>}]
 dyn_cloak_B_01_r [{shapeType=Capsule, length=12.5821, radius=4.4179, shapeName=dyn_cloak_B_01_r_capsule, center=<8.2958 0 3.6>, rotation=<rot 68.3331 -90 -90>}]
 dyn_cloak_B_01_r_MSTR [{shapeType=Capsule, length=0.505, radius=0.505, shapeName=dyn_cloak_B_01_r_MSTR_capsule, rotation=<rot 0 -90 0>}]
 dyn_cloak_B_02_l [{
   shapeType=Capsule
   length=15.0633
   radius=6.4753
   shapeName=dyn_cloak_B_02_l_capsule
   center=<7.6277 1 5>
   rotation=<rot -0.0001 0.0002 112.3075>
  }]
 dyn_cloak_B_02_r [{shapeType=Capsule, length=12.0116, radius=5.4884, shapeName=dyn_cloak_B_02_r_capsule, center=<8.2958 0 3.4>, rotation=<rot 68.3331 -90 -90>}]
 dyn_cloak_B_03_l [{
   shapeType=Capsule
   length=17.2029
   radius=6.8357
   shapeName=dyn_cloak_B_03_l_capsule
   center=<8.2958 1 5.3>
   rotation=<rot -0.0001 0.0002 112.3075>
  }]
 dyn_cloak_B_03_r [{shapeType=Capsule, length=14, radius=6, shapeName=dyn_cloak_B_03_l_capsule, center=<10.3834 0 2.5>, rotation=<rot 68.0724 -80.7122 -81.3734>}]
 dyn_cloak_B_04_l [{
   shapeType=Capsule
   length=28.2284
   radius=9.6636
   shapeName=dyn_cloak_B_04_l_capsule
   center=<9.1008 0.0821 8>
   rotation=<rot -0.0001 0.0004 -65.7856>
  }]
 dyn_cloak_B_04_r [{
   shapeType=Capsule
   length=31.2995
   radius=9.6636
   shapeName=dyn_cloak_B_04_r_capsule
   center=<8.1714 -0.7446 6.5>
   rotation=<rot -0.0002 -0.0007 -106.8475>
  }]
 dyn_cloak_B_05_l [{
   shapeType=Capsule
   length=28.2284
   radius=9.6636
   shapeName=dyn_cloak_B_05_l_capsule
   center=<8.1714 0.2778 7.6>
   rotation=<rot -0.0001 0.0005 -69.1688>
  }]
 dyn_cloak_B_05_r [{
   shapeType=Capsule
   length=31.2995
   radius=9.6636
   shapeName=dyn_cloak_B_05_r_capsule
   center=<8.4314 -1.0675 6.4>
   rotation=<rot -0.0002 -0.0007 -106.8475>
  }]
 dyn_cloak_B_06_l [{
   shapeType=Capsule
   length=28.2284
   radius=9.6636
   shapeName=dyn_cloak_B_06_l_capsule
   center=<8.1714 0.6226 4.8>
   rotation=<rot -0.0001 0.0005 -69.1688>
  }]
 dyn_cloak_B_06_r [{
   shapeType=Capsule
   length=31.2995
   radius=9.6636
   shapeName=dyn_cloak_B_06_r_capsule
   center=<8.4314 -1.4292 5.7>
   rotation=<rot -0.0002 -0.0007 -106.8475>
  }]
 dyn_cloak_B_07_l [{
   shapeType=Capsule
   length=28.2284
   radius=9.6636
   shapeName=dyn_cloak_B_07_l_capsule
   center=<8.1714 0 0.62>
   rotation=<rot -0.0001 0.0005 -69.1688>
  }]
 dyn_cloak_B_07_r [{
   shapeType=Capsule
   length=31.2995
   radius=9.6636
   shapeName=dyn_cloak_C_06_r_capsule
   center=<8.1714 -1.3748 3.6398>
   rotation=<rot -0.0002 -0.0007 -106.8475>
  }]
 dyn_cloak_C_01_l [{shapeType=Capsule, length=12.3538, radius=4.6462, shapeName=dyn_cloak_C_01_l_capsule, center=<8.2958 0 2.1>, rotation=<rot 90 0 0>}]
 dyn_cloak_C_01_l_MSTR [{shapeType=Capsule, length=0.505, radius=0.505, shapeName=dyn_cloak_C_01_l_MSTR_capsule, rotation=<rot 0 -90 0>}]
 dyn_cloak_C_01_r [{shapeType=Capsule, length=12.3518, radius=4.6482, shapeName=dyn_cloak_C_01_r_capsule, center=<8.2958 0 1.6>, rotation=<rot 90 0 0>}]
 dyn_cloak_C_01_r_MSTR [{shapeType=Capsule, length=0.505, radius=0.505, shapeName=dyn_cloak_C_01_r_MSTR_capsule, rotation=<rot 0 -90 0>}]
 dyn_cloak_C_02_l [{shapeType=Capsule, length=14, radius=6, shapeName=dyn_cloak_C_02_l_capsule, center=<8.2958 0 2.4>, rotation=<rot 90 0 0>}]
 dyn_cloak_C_02_r [{shapeType=Capsule, length=14, radius=6, shapeName=dyn_cloak_C_02_r_capsule, center=<8.2958 0 1.9>, rotation=<rot 90 0 0>}]
 dyn_cloak_C_03_l [{
   shapeType=Capsule
   length=28.7445
   radius=9.6636
   shapeName=dyn_cloak_C_03_l_capsule
   center=<8.1714 0.1902 2.3512>
   rotation=<rot 0 0 -102.5793>
  }]
 dyn_cloak_C_03_r [{
   shapeType=Capsule
   length=28.5832
   radius=9.6636
   shapeName=dyn_cloak_B_04_r_capsule
   center=<8.1714 -0.4418 4.4689>
   rotation=<rot -0.0002 -0.0002 -65.5443>
  }]
 dyn_cloak_C_04_l [{
   shapeType=Capsule
   length=28.7445
   radius=9.6636
   shapeName=dyn_cloak_C_04_l_capsule
   center=<8.1714 0.5902 2.3512>
   rotation=<rot 0 0 -102.5793>
  }]
 dyn_cloak_C_04_r [{
   shapeType=Capsule
   length=28.5832
   radius=9.6636
   shapeName=dyn_cloak_C_04_r_capsule
   center=<8.3827 -0.4418 4.4689>
   rotation=<rot -0.0002 -0.0002 -65.5443>
  }]
 dyn_cloak_C_05_l [{
   shapeType=Capsule
   length=28.7445
   radius=9.6636
   shapeName=dyn_cloak_C_05_l_capsule
   center=<8.1714 0.5902 2.3512>
   rotation=<rot 0 0 -102.5793>
  }]
 dyn_cloak_C_05_r [{
   shapeType=Capsule
   length=28.5832
   radius=9.6636
   shapeName=dyn_cloak_C_05_r_capsule
   center=<8.3827 -0.4418 4.4689>
   rotation=<rot -0.0002 -0.0002 -65.5443>
  }]
 dyn_cloak_C_06_l [{
   shapeType=Capsule
   length=28.7445
   radius=9.6636
   shapeName=dyn_cloak_C_06_l_capsule
   center=<8.1714 0.5471 0.7518>
   rotation=<rot 0 0 -102.5793>
  }]
 dyn_cloak_C_06_r [{
   shapeType=Capsule
   length=28.5832
   radius=9.6636
   shapeName=dyn_cloak_C_06_r_capsule
   center=<8.1714 -1.1202 2.9773>
   rotation=<rot -0.0002 -0.0002 -65.5443>
  }]
 dyn_cloak_D_01 [{shapeType=Capsule, length=12.3813, radius=4.6187, shapeName=dyn_cloak_D_01_capsule, center=<8.2958 1.5854 3.2>, rotation=<rot 90 0 0>}]
 dyn_cloak_D_02 [{shapeType=Capsule, length=14, radius=6, shapeName=dyn_cloak_D_02_capsule, center=<8.2958 1.5854 3.5>, rotation=<rot 90 0 0>}]
 dyn_cloak_D_03 [{shapeType=Capsule, length=32.7853, radius=9.6636, shapeName=dyn_cloak_D_03_capsule, center=<8.5984 0 5.4935>, rotation=<rot 0 0 -90>}]
 dyn_cloak_D_04 [{shapeType=Capsule, length=32.7853, radius=9.6636, shapeName=dyn_cloak_D_04_capsule, center=<8.5984 0 5.4935>, rotation=<rot 0 0 -90>}]
 dyn_cloak_D_05 [{shapeType=Capsule, length=32.7853, radius=9.6636, shapeName=dyn_cloak_D_05_capsule, center=<8.5984 0 5.4935>, rotation=<rot 0 0 -90>}]
 dyn_cloak_D_06 [{shapeType=Capsule, length=32.7853, radius=9.6636, shapeName=dyn_cloak_D_06_capsule, center=<8.5984 0 3.8548>, rotation=<rot 0 0 -90>}]
 dyn_cloak_D_08_MSTR [{shapeType=Capsule, length=0.505, radius=0.505, shapeName=dyn_cloak_D_08_MSTR_capsule, rotation=<rot 0 -90 0>}]
 dyn_lightsaber [{shapeType=Capsule, length=20.932, radius=8.4346, shapeName=dyn_lightsaber_capsule, center=<8.7203 0 -6.1176>, rotation=<rot 90 0 0>}]
 dyn_sash_l_01 [{
   shapeType=Capsule
   length=8.0967
   radius=7.9033
   shapeName=dyn_sash_l_01_capsule
   center=<8.1516 -2.3872 -5.3186>
   rotation=<rot 82.8279 -180 -180>
  }]
 dyn_sash_l_02 [{shapeType=Capsule, length=12.1488, radius=8.8622, shapeName=dyn_sash_l_02_capsule, center=<5.1796 -2.0065 -6.3282>, rotation=<rot 90 0 0>}]
 dyn_sash_l_03 [{shapeType=Capsule, length=6.5452, radius=9.4548, shapeName=dyn_sash_l_03_capsule, center=<5.1796 -2.398 -6.1087>, rotation=<rot 90 0 0>}]
 dyn_sash_r_01 [{
   shapeType=Capsule
   length=9.7774
   radius=7.1843
   shapeName=dyn_sash_r_01_capsule
   center=<-8.1909 0.7478 5.2588>
   rotation=<rot 85.7353 -180 -180>
  }]
 dyn_sash_r_02 [{
   shapeType=Capsule
   length=15.7623
   radius=7.9252
   shapeName=dyn_sash_r_02_capsule
   center=<-5.765 1.0832 5.7875>
   rotation=<rot 90 180 -178.8631>
  }]
 dyn_sash_r_03 [{
   shapeType=Capsule
   length=6.9983
   radius=9.2307
   shapeName=dyn_sash_r_03_capsule
   center=<-5.4725 1.5382 6.0396>
   rotation=<rot 87.4229 0 1.8577>
  }]
 dyn_skirt_ground_collision [{shapeType=Box, extentX=318.8156, extentY=318.8156, extentZ=283.5656, shapeName=dyn_skirt_ground_collision_box, center=<0 0 -140>}]
 dyn_underskirt_A_l_01 [{
   shapeType=Capsule
   length=34.7122
   radius=8.1199
   shapeName=dyn_underskirt_C_l_02_capsule
   center=<20.0846 -7.842 -5.1325>
   rotation=<rot 66.3418 97.2104 97.8641>
  }]
 dyn_underskirt_A_r_01 [{
   shapeType=Capsule
   length=33.8478
   radius=8.9843
   shapeName=dyn_underskirt_C_r_02_capsule
   center=<-21.1799 4.518 6.061>
   rotation=<rot 88.0981 178.1801 -178.2316>
  }]
 dyn_underskirt_B_l_02 [{
   shapeType=Capsule
   length=34.6106
   radius=8.2215
   shapeName=dyn_underskirt_B_l_02_capsule
   center=<10.964 -2.5456 -5.613>
   rotation=<rot 87.4497 -0.001 -0.0013>
  }]
 dyn_underskirt_B_r_02 [{
   shapeType=Capsule
   length=35.4698
   radius=7.3623
   shapeName=dyn_underskirt_B_r_02_capsule
   center=<-10.457 5.2883 4.6774>
   rotation=<rot 86.6494 -0.0005 -0.0008>
  }]
 dyn_underskirt_C_l_02 [{
   shapeType=Capsule
   length=32.996
   radius=9.8361
   shapeName=dyn_underskirt_C_l_02_capsule
   center=<10.125 -3.4168 -6.5012>
   rotation=<rot 90 0 0>
  }]
 dyn_underskirt_C_r_02 [{
   shapeType=Capsule
   length=32.7162
   radius=10.1159
   shapeName=dyn_underskirt_C_r_02_capsule
   center=<-10.3884 -1.8207 7.1728>
   rotation=<rot 86.0653 -180 -180>
  }]
 dyn_underskirt_D_02 [{
   shapeType=Capsule
   length=29.9205
   radius=12.9116
   shapeName=dyn_underskirt_D_02_capsule
   center=<10.1557 -2.0914 -8.2538>
   rotation=<rot 88.308 180 180>
  }]
 foot_l [{
   shapeType=Capsule
   length=15.8412
   radius=10.8191
   shapeName=foot_l_capsule
   center=<-3.7001 7.2258 -0.0491>
   rotation=<rot -0.486 1.4485 -88.1788>
  }]
 foot_r [{
   shapeType=Capsule
   length=15.8371
   radius=10.8936
   shapeName=foot_r_capsule
   center=<3.6197 -6.7126 0.2071>
   rotation=<rot -0.5554 1.423 -87.9187>
  }]
 pelvis [{shapeType=Capsule, length=8.0966, radius=14.3791, shapeName=pelvis_capsule}]
 spine_01 [{shapeType=Capsule, length=9.0109, radius=13.4648, shapeName=pelvis_capsule}]
 spine_02 [{shapeType=Capsule, length=12.4566, radius=12.5391, shapeName=spine_02_capsule}]
 spine_03 [{shapeType=Capsule, length=14.466, radius=12.2698, shapeName=spine_03_capsule}]
 thigh_l [{shapeType=Capsule, length=35.9607, radius=9.7682, shapeName=thigh_l_capsule, center=<-14.2913 0 0>, rotation=<rot 85.8185 0.0002 0.0005>}]
 thigh_r [{
   shapeType=Capsule
   length=35.9607
   radius=9.7682
   shapeName=thigh_r_capsule
   center=<14.2913 0 0>
   rotation=<rot -85.8185 -179.9998 179.9995>
  }]
}
constraints [
 {bone1Name=dyn_cloak_A_01_l, bone2Name=dyn_cloak_A_01_l_MSTR, swing1LimitDegrees=2, swing2LimitDegrees=8, twistLimitDegrees=2}
 {bone1Name=dyn_cloak_A_02_l, bone2Name=dyn_cloak_A_01_l, swing1LimitDegrees=2, swing2LimitDegrees=8, twistLimitDegrees=2}
 {bone1Name=dyn_cloak_A_03_l, bone2Name=dyn_cloak_A_02_l, swing1LimitDegrees=2, swing2LimitDegrees=8, twistLimitDegrees=2}
 {bone1Name=dyn_cloak_A_04_l, bone2Name=dyn_cloak_A_03_l, swing1LimitDegrees=8, swing2LimitDegrees=24, twistLimitDegrees=2}
 {bone1Name=dyn_cloak_A_05_l, bone2Name=dyn_cloak_A_04_l, swing1LimitDegrees=8, swing2LimitDegrees=24, twistLimitDegrees=2}
 {bone1Name=dyn_cloak_A_06_l, bone2Name=dyn_cloak_A_05_l, swing1LimitDegrees=8, swing2LimitDegrees=24, twistLimitDegrees=2}
 {bone1Name=dyn_cloak_A_07_l, bone2Name=dyn_cloak_A_06_l, swing1LimitDegrees=8, swing2LimitDegrees=24, twistLimitDegrees=2}
 {bone1Name=dyn_cloak_A_01_r, bone2Name=dyn_cloak_A_01_r_MSTR, swing1LimitDegrees=2, swing2LimitDegrees=8, twistLimitDegrees=2}
 {bone1Name=dyn_cloak_A_02_r, bone2Name=dyn_cloak_A_01_r, swing1LimitDegrees=2, swing2LimitDegrees=8, twistLimitDegrees=2}
 {bone1Name=dyn_cloak_A_03_r, bone2Name=dyn_cloak_A_02_r, swing1LimitDegrees=2, swing2LimitDegrees=8, twistLimitDegrees=2}
 {bone1Name=dyn_cloak_A_04_r, bone2Name=dyn_cloak_A_03_r, swing1LimitDegrees=8, swing2LimitDegrees=24, twistLimitDegrees=2}
 {bone1Name=dyn_cloak_A_05_r, bone2Name=dyn_cloak_A_04_r, swing1LimitDegrees=8, swing2LimitDegrees=24, twistLimitDegrees=2}
 {bone1Name=dyn_cloak_A_06_r, bone2Name=dyn_cloak_A_05_r, swing1LimitDegrees=8, swing2LimitDegrees=24, twistLimitDegrees=2}
 {bone1Name=dyn_cloak_A_07_r, bone2Name=dyn_cloak_A_06_r, swing1LimitDegrees=8, swing2LimitDegrees=24, twistLimitDegrees=2}
 {bone1Name=dyn_cloak_B_01_l, bone2Name=dyn_cloak_B_01_l_MSTR, swing1LimitDegrees=2, swing2LimitDegrees=8, twistLimitDegrees=2}
 {bone1Name=dyn_cloak_B_02_l, bone2Name=dyn_cloak_B_01_l, swing1LimitDegrees=2, swing2LimitDegrees=8, twistLimitDegrees=2}
 {bone1Name=dyn_cloak_B_03_l, bone2Name=dyn_cloak_B_02_l, swing1LimitDegrees=2, swing2LimitDegrees=8, twistLimitDegrees=2}
 {bone1Name=dyn_cloak_B_04_l, bone2Name=dyn_cloak_B_03_l, swing1LimitDegrees=8, swing2LimitDegrees=24, twistLimitDegrees=2}
 {bone1Name=dyn_cloak_B_05_l, bone2Name=dyn_cloak_B_04_l, swing1LimitDegrees=8, swing2LimitDegrees=24, twistLimitDegrees=2}
 {bone1Name=dyn_cloak_B_06_l, bone2Name=dyn_cloak_B_05_l, swing1LimitDegrees=8, swing2LimitDegrees=24, twistLimitDegrees=2}
 {bone1Name=dyn_cloak_B_07_l, bone2Name=dyn_cloak_B_06_l, swing1LimitDegrees=8, swing2LimitDegrees=24, twistLimitDegrees=2}
 {bone1Name=dyn_cloak_B_01_r, bone2Name=dyn_cloak_B_01_r_MSTR, swing1LimitDegrees=2, swing2LimitDegrees=8, twistLimitDegrees=2}
 {bone1Name=dyn_cloak_B_03_r, bone2Name=dyn_cloak_B_02_r, swing1LimitDegrees=2, swing2LimitDegrees=8, twistLimitDegrees=2}
 {bone1Name=dyn_cloak_B_04_r, bone2Name=dyn_cloak_B_03_r, swing1LimitDegrees=8, swing2LimitDegrees=24, twistLimitDegrees=2}
 {bone1Name=dyn_cloak_B_05_r, bone2Name=dyn_cloak_B_04_r, swing1LimitDegrees=8, swing2LimitDegrees=24, twistLimitDegrees=2}
 {bone1Name=dyn_cloak_B_06_r, bone2Name=dyn_cloak_B_05_r, swing1LimitDegrees=8, swing2LimitDegrees=24, twistLimitDegrees=2}
 {bone1Name=dyn_cloak_C_01_l, bone2Name=dyn_cloak_C_01_l_MSTR, swing1LimitDegrees=2, swing2LimitDegrees=8, twistLimitDegrees=2}
 {bone1Name=dyn_cloak_C_03_l, bone2Name=dyn_cloak_C_02_l, swing1LimitDegrees=8, swing2LimitDegrees=20, twistLimitDegrees=4}
 {bone1Name=dyn_cloak_C_04_l, bone2Name=dyn_cloak_C_03_l, swing1LimitDegrees=8, swing2LimitDegrees=20, twistLimitDegrees=4}
 {bone1Name=dyn_cloak_C_05_l, bone2Name=dyn_cloak_C_04_l, swing1LimitDegrees=8, swing2LimitDegrees=20, twistLimitDegrees=4}
 {bone1Name=dyn_cloak_C_06_l, bone2Name=dyn_cloak_C_05_l, swing1LimitDegrees=8, swing2LimitDegrees=20, twistLimitDegrees=4}
 {bone1Name=dyn_cloak_C_01_r, bone2Name=dyn_cloak_C_01_r_MSTR, swing1LimitDegrees=2, swing2LimitDegrees=8, twistLimitDegrees=2}
 {bone1Name=dyn_cloak_C_02_r, bone2Name=dyn_cloak_C_01_r, swing1LimitDegrees=2, swing2LimitDegrees=8, twistLimitDegrees=2}
 {bone1Name=dyn_cloak_C_03_r, bone2Name=dyn_cloak_C_02_r, swing1LimitDegrees=8, swing2LimitDegrees=20, twistLimitDegrees=4}
 {bone1Name=dyn_cloak_C_04_r, bone2Name=dyn_cloak_C_03_r, swing1LimitDegrees=8, swing2LimitDegrees=20, twistLimitDegrees=4}
 {bone1Name=dyn_cloak_C_05_r, bone2Name=dyn_cloak_C_04_r, swing1LimitDegrees=8, swing2LimitDegrees=20, twistLimitDegrees=4}
 {bone1Name=dyn_cloak_C_06_r, bone2Name=dyn_cloak_C_05_r, swing1LimitDegrees=8, swing2LimitDegrees=20, twistLimitDegrees=4}
 {bone1Name=dyn_cloak_D_01, bone2Name=dyn_cloak_D_08_MSTR, swing1LimitDegrees=2, swing2LimitDegrees=8, twistLimitDegrees=2}
 {bone1Name=dyn_cloak_D_03, bone2Name=dyn_cloak_D_02, swing1LimitDegrees=8, swing2LimitDegrees=24, twistLimitDegrees=4}
 {bone1Name=dyn_cloak_D_04, bone2Name=dyn_cloak_D_03, swing1LimitDegrees=8, swing2LimitDegrees=24, twistLimitDegrees=4}
 {bone1Name=dyn_cloak_D_05, bone2Name=dyn_cloak_D_04, swing1LimitDegrees=8, swing2LimitDegrees=24, twistLimitDegrees=4}
 {bone1Name=dyn_cloak_D_06, bone2Name=dyn_cloak_D_05, swing1LimitDegrees=8, swing2LimitDegrees=24, twistLimitDegrees=4}
 {bone1Name=dyn_cloak_B_02_r, bone2Name=dyn_cloak_B_01_r, swing1LimitDegrees=2, swing2LimitDegrees=8, twistLimitDegrees=2}
 {bone1Name=dyn_cloak_C_02_l, bone2Name=dyn_cloak_C_01_l, swing1LimitDegrees=2, swing2LimitDegrees=8, twistLimitDegrees=2}
 {bone1Name=dyn_cloak_D_02, bone2Name=dyn_cloak_D_01, swing1LimitDegrees=2, swing2LimitDegrees=8, twistLimitDegrees=4}
 {bone1Name=dyn_cloak_B_07_r, bone2Name=dyn_cloak_B_06_r, swing1LimitDegrees=8, swing2LimitDegrees=24, twistLimitDegrees=2}
 {
  bone1Name=dyn_cloak_A_07_r
  bone2Name=dyn_cloak_B_07_r
  swing1LimitDegrees=45
  swing1Motion=Free
  swing2LimitDegrees=45
  swing2Motion=Free
  twistLimitDegrees=45
  twistMotion=Free
 }
 {
  bone1Name=dyn_cloak_A_07_l
  bone2Name=dyn_cloak_B_07_l
  swing1LimitDegrees=45
  swing1Motion=Free
  swing2LimitDegrees=45
  swing2Motion=Free
  twistLimitDegrees=45
  twistMotion=Free
 }
]
