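# Physics asset for the GeneAglet male medium character's FaceAcc part, set up around a lowered hood.
# Three hood capsules form a chain from a tiny dyn_hooddown_MSTR root capsule, linked by constraints with moderate swing and tight twist limits.
# Capsules on the upper spine and both clavicles have no constraints, giving the hood chain a torso and shoulder shape to collide with.
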PhysicsAsset /BRCosmetics/Characters/Player/Male/Medium/Bodies/M_MED_GeneAglet/Meshes/Parts/M_MED_GeneAglet_FaceAcc_Physics
bodies {
 clavicle_l [{shapeType=Capsule, length=7.7997, radius=11.6983, shapeName=clavicle_l_capsule, center=<6.5498 0 0>, rotation=<rot 80.7652 -0.0003 -0.0003>}]
 clavicle_r [{
   shapeType=Capsule
   length=7.7997
   radius=11.6983
   shapeName=clavicle_l_capsule
   center=<-7.6273 0 -0.1545>
   rotation=<rot 80.7652 -0.0003 -0.0003>
  }]
 dyn_hooddown_01 [{shapeType=Capsule, length=17.8941, radius=8.9944, shapeName=dyn_hooddown_01_capsule, center=<0 0 9.7>, rotation=<rot 0 0 90>}]
 dyn_hooddown_02 [{shapeType=Capsule, length=14.2847, radius=10.04, shapeName=dyn_hooddown_02_capsule, center=<0.8831 0 9>, rotation=<rot 0 0 90>}]
 dyn_hooddown_03 [{shapeType=Capsule, length=10.7883, radius=10.1537, shapeName=dyn_hooddown_03_capsule, center=<0 0 5.5>, rotation=<rot 0 0 90>}]
 dyn_hooddown_MSTR [{shapeType=Capsule, length=0.505, radius=0.505, shapeName=dyn_hooddown_MSTR_capsule, rotation=<rot 0 -90 0>}]
 spine_03 [{shapeType=Capsule, length=11.0404, radius=15.2011, shapeName=spine_05_capsule}]
 spine_04 [{shapeType=Capsule, length=11.591, radius=14.6505, shapeName=spine_05_capsule}]
 spine_05 [{shapeType=Capsule, length=16.1564, radius=14.1941, shapeName=spine_05_capsule}]
}
constraints [
 {bone1Name=dyn_hooddown_02, bone2Name=dyn_hooddown_01, swing1LimitDegrees=30, swing2LimitDegrees=30, twistLimitDegrees=4}
 {bone1Name=dyn_hooddown_03, bone2Name=dyn_hooddown_02, swing1LimitDegrees=30, swing2LimitDegrees=30, twistLimitDegrees=4}
 {bone1Name=dyn_hooddown_01, bone2Name=dyn_hooddown_MSTR, swing1LimitDegrees=30, swing2LimitDegrees=30, twistLimitDegrees=4}
]
