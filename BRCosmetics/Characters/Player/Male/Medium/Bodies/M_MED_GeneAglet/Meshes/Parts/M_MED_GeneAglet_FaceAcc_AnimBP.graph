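# Post-process style AnimBlueprint for the GeneAglet male face accessory, adding secondary motion to its hood.
# Copies body and CharacterMesh poses, layered at the upper arms, into a Control Rig, then spring-driven AnimDynamics on three hood-up bones.
# A RigidBody pass follows, bypassed when the backbling check sets isWearingPack.
# Hood-down nodes are disconnected, and the zipline AnimDynAlpha is immediately reset to 1 and never read by the AnimGraph.
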AnimBlueprint /BRCosmetics/Characters/Player/Male/Medium/Bodies/M_MED_GeneAglet/Meshes/Parts/M_MED_GeneAglet_FaceAcc_AnimBP
target_skeleton=&/M_MED_GeneAglet_FaceAcc_Skeleton
graphs [
 AnimGraph {nodes [
   n1 = Anim Dynamics - Bone: dyn_hooddown_01!(SimulationSpace:AnimPhysSimSpaceType Enum=Component, ComponentPose, Alpha=1, ExternalForce="0,0,0") -> (Pose) {
    angularDampingOverride=0.999
    angularSpringConstant=500
    bOverrideAngularDamping=true
    bUseGravityOverride=true
    alphaBoolBlend {blendOption=Linear}
    alphaScaleBiasClamp {bInterpResult=true}
    boundBone {boneName=dyn_hooddown_01}
    componentAppliedLinearAccClamp=<10000 10000 10000>
    componentLinearAccScale=<0.4 1 0.4>
    componentLinearVelScale=<0.4 1.25 0.4>
    gravityOverride {z=-980}
    physicsBodyDefinitions [{
      boundBone {boneName=dyn_hooddown_01}
      boxExtents=<12 12 12>
      constraintSetup {angularLimitsMax=<0 0 0>, angularLimitsMin=<0 -60 0>, angularTarget=<1 0 0>}
      localJointOffset=<1 0 0>
     }]
   }
   n2 = Anim Dynamics - Bone: dyn_hooddown_02!(SimulationSpace:AnimPhysSimSpaceType Enum=Component, ComponentPose=@n1, Alpha=1, ExternalForce="0,0,0") -> (Pose) {
    angularDampingOverride=0.999
    angularSpringConstant=500
    bOverrideAngularDamping=true
    bUseGravityOverride=true
    alphaBoolBlend {blendOption=Linear}
    alphaScaleBiasClamp {bInterpResult=true}
    boundBone {boneName=dyn_hooddown_02}
    componentAppliedLinearAccClamp=<10000 10000 10000>
    componentLinearAccScale=<0.4 1 0.4>
    componentLinearVelScale=<0.4 1.25 0.4>
    gravityOverride {z=-980}
    physicsBodyDefinitions [{
      boundBone {boneName=dyn_hooddown_02}
      boxExtents=<12 12 12>
      constraintSetup {angularLimitsMax=<0 0 0>, angularLimitsMin=<0 -60 0>, angularTarget=<1 0 0>}
      localJointOffset=<1 0 0>
     }]
   }
   Anim Dynamics - Bone: dyn_hooddown_03!(SimulationSpace:AnimPhysSimSpaceType Enum=Component, ComponentPose=@n2, Alpha=1, ExternalForce="0,0,0") -> (Pose) {
    angularDampingOverride=0.999
    angularSpringConstant=500
    bOverrideAngularDamping=true
    bUseGravityOverride=true
    alphaBoolBlend {blendOption=Linear}
    alphaScaleBiasClamp {bInterpResult=true}
    boundBone {boneName=dyn_hooddown_03}
    componentAppliedLinearAccClamp=<10000 10000 10000>
    componentLinearAccScale=<0.4 1 0.4>
    componentLinearVelScale=<0.4 1.25 0.4>
    gravityOverride {z=-980}
    physicsBodyDefinitions [{
      boundBone {boneName=dyn_hooddown_03}
      boxExtents=<12 12 12>
      constraintSetup {angularLimitsMax=<0 0 0>, angularLimitsMin=<0 -60 0>, angularTarget=<1 0 0>}
      localJointOffset=<1 0 0>
     }]
   }
   "Transform (Modify) Bone - Bone: dyn_hooddown_MSTR"!(ComponentPose, bAlphaBoolEnabled:bool=True, Translation="0,0,1", Rotation="2,0,0", Scale="1,1,1") -> (Pose) {
    alphaInputType=Bool
    rotationMode=BMM_Additive
    rotationSpace=BCS_BoneSpace
    translationMode=BMM_Additive
    translationSpace=BCS_BoneSpace
    alphaBoolBlend {blendInTime=1, blendOption=Linear, blendOutTime=1}
    boneToModify {boneName=dyn_hooddown_MSTR}
    rotation {pitch=2}
    translation {z=1}
   }
   Output Pose(Result=Use cached pose 'FinalCache'()) {layerGroup=None}
   n3 = Get CharacterMesh() -> (CharacterMesh:Skeletal Mesh Component Object Reference)
   n4 = Control Rig(Source=Layered blend per bone(
     BasePose=Copy Pose From Mesh() {bUseAttachedParent=true}
     BlendPoses_0=Copy Pose From Mesh(SourceMeshComponent=@n3) {bUseAttachedParent=true}
    ) {layerSetup [{branchFilters [{boneName=upperarm_l}, {boneName=upperarm_r}]}]}) {
    alphaBoolBlend {blendOption=Linear}
    controlRigAssetReference {blueprintRigClass=&/M_MED_GeneAglet_FaceAcc_CtrlRig.M_MED_GeneAglet_FaceAcc_CtrlRig_C}
   }
   n5 = Anim Dynamics - Bone: dyn_hoodup_l!(SimulationSpace:AnimPhysSimSpaceType Enum, ComponentPose=Local To Component(LocalPose=@n4), Alpha=1, ExternalForce="0,0,0") -> (Pose) {
    angularDampingOverride=0.9999
    angularSpringConstant=300
    bLinearSpring=true
    bOverrideAngularDamping=true
    bOverrideLinearDamping=true
    bUseGravityOverride=true
    linearDampingOverride=0.9999
    linearSpringConstant=50
    simulationSpace=RootRelative
    alphaBoolBlend {blendOption=Linear}
    boundBone {boneName=dyn_hoodup_l}
    componentAppliedLinearAccClamp=<10000 10000 10000>
    componentLinearAccScale=<0.3 0.3 0.3>
    componentLinearVelScale=<0.4 0.4 0.4>
    gravityOverride {z=-980}
    physicsBodyDefinitions [{
      boundBone {boneName=dyn_hoodup_l}
      boxExtents=<14 14 14>
      constraintSetup {angularLimitsMax=<0 0 0>, angularLimitsMin=<0 0 0>, angularTarget=<1 0 0>, linearAxesMax=<6 4 2>, linearAxesMin=<-6 -4 -2>}
      localJointOffset=<2 0 0>
     }]
   }
   n6 = Anim Dynamics - Bone: dyn_hoodup_r!(SimulationSpace:AnimPhysSimSpaceType Enum, ComponentPose=@n5, Alpha=1, ExternalForce="0,0,0") -> (Pose) {
    angularDampingOverride=0.9999
    angularSpringConstant=300
    bLinearSpring=true
    bOverrideAngularDamping=true
    bOverrideLinearDamping=true
    bUseGravityOverride=true
    linearDampingOverride=0.9999
    linearSpringConstant=50
    simulationSpace=RootRelative
    alphaBoolBlend {blendOption=Linear}
    boundBone {boneName=dyn_hoodup_r}
    componentAppliedLinearAccClamp=<10000 10000 10000>
    componentLinearAccScale=<0.3 0.3 0.3>
    componentLinearVelScale=<0.4 0.4 0.4>
    gravityOverride {z=-980}
    physicsBodyDefinitions [{
      boundBone {boneName=dyn_hoodup_r}
      boxExtents=<14 14 14>
      constraintSetup {angularLimitsMax=<0 0 0>, angularLimitsMin=<0 0 0>, angularTarget=<1 0 0>, linearAxesMax=<6 4 1>, linearAxesMin=<-6 -4 -1>}
      localJointOffset=<2 0 0>
     }]
   }
   n7 = Anim Dynamics - Bone: dyn_hoodup_bk!(SimulationSpace:AnimPhysSimSpaceType Enum, ComponentPose=@n6, Alpha=1, ExternalForce="0,0,0") -> (Pose) {
    angularDampingOverride=0.9999
    angularSpringConstant=50
    bAngularSpring=true
    bLinearSpring=true
    bOverrideAngularDamping=true
    bOverrideLinearDamping=true
    bUseGravityOverride=true
    linearDampingOverride=0.9999
    linearSpringConstant=50
    simulationSpace=RootRelative
    alphaBoolBlend {blendOption=Linear}
    boundBone {boneName=dyn_hoodup_bk}
    componentAppliedLinearAccClamp=<10000 10000 10000>
    componentLinearAccScale=<0.3 0.3 0.3>
    componentLinearVelScale=<0.4 0.4 0.4>
    gravityOverride {z=-980}
    physicsBodyDefinitions [{
      boundBone {boneName=dyn_hoodup_bk}
      boxExtents=<14 14 14>
      constraintSetup {angularLimitsMax=<2 2 2>, angularLimitsMin=<-2 -2 -2>, angularTarget=<1 0 0>, linearAxesMax=<6 6 4>, linearAxesMin=<-6 -6 0>}
      localJointOffset=<1 0 0>
     }]
   }
   Save cached pose 'MainCache'(Pose=Component To Local(ComponentPose=@n7)) {cacheName=MainCache}
   n8 = RigidBody(ComponentPose=Local To Component(LocalPose=Use cached pose 'MainCache'())) {
    bClampLinearTranslationLimitToRefPose=true
    bOverrideWorldGravity=true
    overridePhysicsAsset=&/M_MED_GeneAglet_FaceAcc_Physics
    simulationSpace=BaseBoneSpace
    alphaBoolBlend {blendOption=Linear}
    baseBoneRef {boneName=spine_05}
    overrideWorldGravity {z=-980}
    simSpaceSettings {
     dampingAlpha=0
     maxAngularAcceleration=15
     maxAngularVelocity=3
     maxLinearAcceleration=1500
     maxLinearVelocity=300
     velocityScaleZ=0.6
     worldAlpha=1
     externalLinearDragV=<0.5 0.5 0.5>
    }
   }
   Save cached pose 'HoodCache'(Pose=Component To Local(ComponentPose=@n8)) {cacheName=HoodCache}
   Save cached pose 'FinalCache'(Pose=Blend Poses by bool(
     BlendPose_0=Use cached pose 'MainCache'()
     BlendPose_1=Use cached pose 'HoodCache'()
     bActiveValue=Get isWearingPack() -> (isWearingPack:bool)
    ) {blendType=Linear}) {cacheName=FinalCache}
  ]}
 EventGraph {nodes [
   n1 = Gravity Override State Function DEPRECATED(CharacterPartAnimInstance=Self-Reference(), GameTimeInSeconds=GetGameTimeInSeconds(), GravityOverrideParams=Get Gravity Override Params())
   > Set Gravity Override(Gravity Override=@n1)
   Event BlueprintUpdateAnimation()
   > n2 = Sequence!(execute) -> (then_0, then_1, then_2:Exec)
   > n3 = Branch(Condition=Get IsOnZipline() -> (IsOnZipline:bool))
   > Set AnimDynAlpha(AnimDynAlpha:double=0) -> (Output_Get:double)
   Set AnimDynAlpha(execute=@n3.else, AnimDynAlpha=1)
   n4 = Branch(execute=@n2.then_1, Condition=Get bIsInFrontEnd())
   > Set Simulation Space(Simulation Space:AnimPhysSimSpaceType Enum=World) -> (Output_Get:AnimPhysSimSpaceType Enum)
   > Set AnimDynAlpha(AnimDynAlpha=1)
   Set Simulation Space(execute=@n4.else)
   > Set AnimDynAlpha(AnimDynAlpha=1)
   n5 = Event OnCharacterCustomizationLoaded()
   n6 = CheckForBackbling()
   > Set isWearingPack(isWearingPack:bool=OR Boolean!(A=@n6, B=@n6.bIsBackblingACape, C:bool=@n6.bIsBackblingLong) -> (ReturnValue)) -> (Output_Get:bool)
   > Set isWearingCape(isWearingCape=@n6.bIsBackblingLong)
   > Set CharacterMesh(CharacterMesh:Skeletal Mesh Component Object Reference=@n5.CustomBaseMesh) -> (Output_Get:Skeletal Mesh Component Object Reference)
   n7 = Event BlueprintProcessAnimInputEvent()
   > n8 = Branch(Condition=IsNamed(self=@n7.AnimInputEvent, InEventName=aie.Player.Zipline.Entered))
   > Set IsOnZipline(IsOnZipline:bool=true) -> (Output_Get:bool)
   Branch(execute=@n8.else, Condition=IsNamed(self=@n7.AnimInputEvent, InEventName=aie.Player.Zipline.Exited))
   > Set IsOnZipline(IsOnZipline=false)
  ]}
]
variables [
 AnimDynAlpha {type=number, default=1}
 CharacterMesh {type=SkeletalMeshComponent}
 Gravity Override {type=Vector}
 Gravity Override Params {type=GravityOverrideParamsStruct}
 IsOnZipline {type=boolean}
 isWearingCape {type=boolean}
 isWearingPack {type=boolean}
 Simulation Space {type=AnimPhysSimSpaceType, default=RootRelative}
]
